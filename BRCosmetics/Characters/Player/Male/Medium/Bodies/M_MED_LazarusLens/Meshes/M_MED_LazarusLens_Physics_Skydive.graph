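# Physics asset for the M_MED_LazarusLens male character body, simulating its long jacket during skydiving.
# Capsules on pelvis, spine_01 and thighs form the body, joined by limited constraints.
# Three four-segment jacket chains per side and a back chain forking left and right hang from master bones with free angular constraints to spine_01.
# Chain links have tight twist and asymmetric swing limits that generally widen toward the tips.
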PhysicsAsset /BRCosmetics/Characters/Player/Male/Medium/Bodies/M_MED_LazarusLens/Meshes/M_MED_LazarusLens_Physics_Skydive
bodies {
 dyn_jacket_01_L_01 [{shapeType=Capsule, length=8, radius=5, shapeName=dyn_jacket_01_L_01_capsule, center=<3 0 3>, rotation=<rot 57.7517 -90 -0.0001>}]
 dyn_jacket_01_L_02 [{shapeType=Capsule, length=8, radius=6, shapeName=dyn_jacket_01_L_02_capsule, center=<3 0 5>, rotation=<rot 57.7517 -90 -0.0001>}]
 dyn_jacket_01_L_03 [{shapeType=Capsule, length=8, radius=6, shapeName=dyn_jacket_01_L_03_capsule, center=<3 0 4>, rotation=<rot 57.7517 -90 -0.0001>}]
 dyn_jacket_01_L_04 [{shapeType=Capsule, length=15.5, radius=6, shapeName=dyn_jacket_01_L_04_capsule, center=<3 0 4>, rotation=<rot 57.7517 -90 -0.0001>}]
 dyn_jacket_01_L_mstr [{shapeType=Capsule, length=1, radius=1, shapeName=dyn_jacket_01_L_mstr_capsule}]
 dyn_jacket_01_R_01 [{shapeType=Capsule, length=8, radius=6, shapeName=dyn_jacket_01_L_01_capsule, center=<3 0 3>, rotation=<rot 62.7487 90.0001 -179.9999>}]
 dyn_jacket_01_R_02 [{shapeType=Capsule, length=8, radius=6, shapeName=dyn_jacket_01_L_02_capsule, center=<3 0 4>, rotation=<rot 62.7487 90.0001 -179.9999>}]
 dyn_jacket_01_R_03 [{shapeType=Capsule, length=8, radius=6, shapeName=dyn_jacket_01_L_03_capsule, center=<3 0 4>, rotation=<rot 62.7487 90.0001 -179.9999>}]
 dyn_jacket_01_R_04 [{shapeType=Capsule, length=10, radius=6, shapeName=dyn_jacket_01_L_04_capsule, center=<3 0 4>, rotation=<rot 62.7487 90.0001 -179.9999>}]
 dyn_jacket_01_R_mstr [{shapeType=Capsule, length=1, radius=1, shapeName=dyn_jacket_01_L_mstr_capsule}]
 dyn_jacket_02_L_01 [{shapeType=Capsule, length=8, radius=6, shapeName=dyn_jacket_02_L_01_capsule, center=<3 0 3>, rotation=<rot 79.2586 -90 0>}]
 dyn_jacket_02_L_02 [{shapeType=Capsule, length=8, radius=6, shapeName=dyn_jacket_02_L_02_capsule, center=<3 0 3>, rotation=<rot 79.2586 -90 0>}]
 dyn_jacket_02_L_03 [{shapeType=Capsule, length=8, radius=6, shapeName=dyn_jacket_02_L_03_capsule, center=<3 0 3>, rotation=<rot 79.2586 -90 0>}]
 dyn_jacket_02_L_04 [{shapeType=Capsule, length=15.5, radius=6, shapeName=dyn_jacket_02_L_04_capsule, center=<3 0 3>, rotation=<rot 79.2586 -90 0>}]
 dyn_jacket_02_L_mstr [{shapeType=Capsule, length=1, radius=1, shapeName=dyn_jacket_02_L_mstr_capsule}]
 dyn_jacket_02_R_01 [{shapeType=Capsule, length=8, radius=6, shapeName=dyn_jacket_02_L_01_capsule, center=<3 0 3>, rotation=<rot 80.7238 89.9999 180>}]
 dyn_jacket_02_R_02 [{shapeType=Capsule, length=8, radius=6, shapeName=dyn_jacket_02_L_02_capsule, center=<3 0 3>, rotation=<rot 80.7238 89.9999 180>}]
 dyn_jacket_02_R_03 [{shapeType=Capsule, length=8, radius=6, shapeName=dyn_jacket_02_L_03_capsule, center=<3 0 3>, rotation=<rot 80.7238 89.9999 180>}]
 dyn_jacket_02_R_04 [{shapeType=Capsule, length=15.5, radius=6, shapeName=dyn_jacket_02_L_04_capsule, center=<3 0 3>, rotation=<rot 80.7238 89.9999 180>}]
 dyn_jacket_02_R_mstr [{shapeType=Capsule, length=1, radius=1, shapeName=dyn_jacket_01_L_mstr_capsule}]
 dyn_jacket_03_L_01 [{shapeType=Capsule, length=8, radius=7, shapeName=dyn_jacket_03_L_01_capsule, center=<3 0 5>, rotation=<rot 90 0 90>}]
 dyn_jacket_03_L_02 [{shapeType=Capsule, length=8, radius=7.5, shapeName=dyn_jacket_03_L_02_capsule, center=<3 0 5>, rotation=<rot 90 0 90>}]
 dyn_jacket_03_L_03 [{shapeType=Capsule, length=8, radius=7.5, shapeName=dyn_jacket_03_L_03_capsule, center=<3 0 5>, rotation=<rot 90 0 90>}]
 dyn_jacket_03_L_04 [{shapeType=Capsule, length=8, radius=7.5, shapeName=dyn_jacket_03_L_04_capsule, center=<3 0 5>, rotation=<rot 90 0 90>}]
 dyn_jacket_03_L_mstr [{shapeType=Capsule, length=1, radius=1, shapeName=dyn_jacket_03_L_mstr_capsule}]
 dyn_jacket_03_R_01 [{shapeType=Capsule, length=8, radius=7, shapeName=dyn_jacket_01_L_01_capsule, center=<3 0 5>, rotation=<rot 90 0 90>}]
 dyn_jacket_03_R_02 [{shapeType=Capsule, length=8, radius=7, shapeName=dyn_jacket_01_L_01_capsule, center=<3 0 5>, rotation=<rot 90 0 90>}]
 dyn_jacket_03_R_03 [{shapeType=Capsule, length=8, radius=7.5, shapeName=dyn_jacket_01_L_01_capsule, center=<3 0 5>, rotation=<rot 90 0 90>}]
 dyn_jacket_03_R_04 [{shapeType=Capsule, length=8, radius=7.5, shapeName=dyn_jacket_01_L_01_capsule, center=<3 0 5>, rotation=<rot 90 0 90>}]
 dyn_jacket_03_R_mstr [{shapeType=Capsule, length=1, radius=1, shapeName=dyn_jacket_01_L_mstr_capsule}]
 dyn_jacket_B_01 [{shapeType=Capsule, length=10, radius=7.5, shapeName=dyn_jacket_B_01_capsule, center=<3 0 5>, rotation=<rot 90 0 90>}]
 dyn_jacket_B_02 [{shapeType=Capsule, length=10, radius=7.5, shapeName=dyn_jacket_B_02_capsule, center=<3 0 5>, rotation=<rot 90 0 90>}]
 dyn_jacket_B_L [{shapeType=Capsule, length=10, radius=9, shapeName=dyn_jacket_B_L_capsule, center=<3 0 5>, rotation=<rot 90 0 90>}]
 dyn_jacket_B_R [{shapeType=Capsule, length=10, radius=9, shapeName=dyn_jacket_B_L_capsule, center=<3 0 5>, rotation=<rot 90 0 90>}]
 dyn_jacket_B_mstr [{shapeType=Capsule, length=1, radius=1, shapeName=dyn_jacket_B_mstr_capsule}]
 pelvis [{shapeType=Capsule, length=12.6372, radius=13, center=<6.4853 -1.5035 -0.0584>, rotation=<rot 2.232 -90.0001 -89.1235>}]
 spine_01 [{shapeType=Capsule, length=12, radius=12, center=<8.4965 0.7002 -0.0045>, rotation=<rot 0 90 90>}]
 thigh_l [{shapeType=Capsule, length=25, radius=9, center=<-18.002 0 0>, rotation=<rot 90 0 0>}]
 thigh_r [{shapeType=Capsule, length=25, radius=9, center=<18.002 0 0>, rotation=<rot 90 0 0>}]
}
constraints [
 {bone1Name=spine_01, bone2Name=pelvis, swing1LimitDegrees=45, swing2LimitDegrees=45, twistLimitDegrees=45}
 {bone1Name=thigh_l, bone2Name=pelvis, swing1LimitDegrees=45, swing2LimitDegrees=45, twistLimitDegrees=45}
 {bone1Name=thigh_r, bone2Name=pelvis, swing1LimitDegrees=45, swing2LimitDegrees=45, twistLimitDegrees=45}
 {
  bone1Name=dyn_jacket_01_L_mstr
  bone2Name=spine_01
  swing1LimitDegrees=45
  swing1Motion=Free
  swing2LimitDegrees=45
  swing2Motion=Free
  twistLimitDegrees=45
  twistMotion=Free
 }
 {bone1Name=dyn_jacket_01_L_01, bone2Name=dyn_jacket_01_L_mstr, swing1LimitDegrees=5, swing2LimitDegrees=30, twistLimitDegrees=2}
 {bone1Name=dyn_jacket_01_L_02, bone2Name=dyn_jacket_01_L_01, swing1LimitDegrees=5, swing2LimitDegrees=30, twistLimitDegrees=2}
 {bone1Name=dyn_jacket_01_L_03, bone2Name=dyn_jacket_01_L_02, swing1LimitDegrees=5, swing2LimitDegrees=35, twistLimitDegrees=2}
 {bone1Name=dyn_jacket_01_L_04, bone2Name=dyn_jacket_01_L_03, swing1LimitDegrees=5, swing2LimitDegrees=45, twistLimitDegrees=2}
 {
  bone1Name=dyn_jacket_02_L_mstr
  bone2Name=spine_01
  swing1LimitDegrees=45
  swing1Motion=Free
  swing2LimitDegrees=45
  swing2Motion=Free
  twistLimitDegrees=45
  twistMotion=Free
 }
 {bone1Name=dyn_jacket_02_L_01, bone2Name=dyn_jacket_02_L_mstr, swing1LimitDegrees=5, swing2LimitDegrees=20, twistLimitDegrees=2}
 {bone1Name=dyn_jacket_02_L_02, bone2Name=dyn_jacket_02_L_01, swing1LimitDegrees=5, swing2LimitDegrees=20, twistLimitDegrees=2}
 {bone1Name=dyn_jacket_02_L_03, bone2Name=dyn_jacket_02_L_02, swing1LimitDegrees=10, swing2LimitDegrees=25, twistLimitDegrees=2}
 {bone1Name=dyn_jacket_02_L_04, bone2Name=dyn_jacket_02_L_03, swing1LimitDegrees=15, swing2LimitDegrees=30, twistLimitDegrees=2}
 {
  bone1Name=dyn_jacket_03_L_mstr
  bone2Name=spine_01
  swing1LimitDegrees=45
  swing1Motion=Free
  swing2LimitDegrees=45
  swing2Motion=Free
  twistLimitDegrees=45
  twistMotion=Free
 }
 {bone1Name=dyn_jacket_03_L_01, bone2Name=dyn_jacket_03_L_mstr, swing1LimitDegrees=10, swing2LimitDegrees=20, twistLimitDegrees=2}
 {bone1Name=dyn_jacket_03_L_02, bone2Name=dyn_jacket_03_L_01, swing1LimitDegrees=10, swing2LimitDegrees=20, twistLimitDegrees=2}
 {bone1Name=dyn_jacket_03_L_03, bone2Name=dyn_jacket_03_L_02, swing1LimitDegrees=10, swing2LimitDegrees=25, twistLimitDegrees=2}
 {bone1Name=dyn_jacket_03_L_04, bone2Name=dyn_jacket_03_L_03, swing1LimitDegrees=15, swing2LimitDegrees=30, twistLimitDegrees=2}
 {
  bone1Name=dyn_jacket_B_mstr
  bone2Name=spine_01
  swing1LimitDegrees=45
  swing1Motion=Free
  swing2LimitDegrees=45
  swing2Motion=Free
  twistLimitDegrees=45
  twistMotion=Free
 }
 {bone1Name=dyn_jacket_B_01, bone2Name=dyn_jacket_B_mstr, swing1LimitDegrees=5, swing2LimitDegrees=20, twistLimitDegrees=2}
 {bone1Name=dyn_jacket_B_02, bone2Name=dyn_jacket_B_01, swing1LimitDegrees=10, swing2LimitDegrees=25, twistLimitDegrees=2}
 {bone1Name=dyn_jacket_B_L, bone2Name=dyn_jacket_B_02, swing1LimitDegrees=30, swing2LimitDegrees=30, twistLimitDegrees=2}
 {bone1Name=dyn_jacket_01_R_04, bone2Name=dyn_jacket_01_R_03, swing1LimitDegrees=15, swing2LimitDegrees=30, twistLimitDegrees=2}
 {
  bone1Name=dyn_jacket_01_R_mstr
  bone2Name=spine_01
  swing1LimitDegrees=45
  swing1Motion=Free
  swing2LimitDegrees=45
  swing2Motion=Free
  twistLimitDegrees=45
  twistMotion=Free
 }
 {bone1Name=dyn_jacket_01_R_01, bone2Name=dyn_jacket_01_R_mstr, swing1LimitDegrees=5, swing2LimitDegrees=20, twistLimitDegrees=2}
 {bone1Name=dyn_jacket_01_R_02, bone2Name=dyn_jacket_01_R_01, swing1LimitDegrees=5, swing2LimitDegrees=20, twistLimitDegrees=2}
 {bone1Name=dyn_jacket_01_R_03, bone2Name=dyn_jacket_01_R_02, swing1LimitDegrees=10, swing2LimitDegrees=25, twistLimitDegrees=2}
 {bone1Name=dyn_jacket_02_R_04, bone2Name=dyn_jacket_02_R_03, swing1LimitDegrees=15, swing2LimitDegrees=30, twistLimitDegrees=2}
 {
  bone1Name=dyn_jacket_02_R_mstr
  bone2Name=spine_01
  swing1LimitDegrees=45
  swing1Motion=Free
  swing2LimitDegrees=45
  swing2Motion=Free
  twistLimitDegrees=45
  twistMotion=Free
 }
 {bone1Name=dyn_jacket_02_R_01, bone2Name=dyn_jacket_02_R_mstr, swing1LimitDegrees=5, swing2LimitDegrees=20, twistLimitDegrees=2}
 {bone1Name=dyn_jacket_02_R_02, bone2Name=dyn_jacket_02_R_01, swing1LimitDegrees=5, swing2LimitDegrees=20, twistLimitDegrees=2}
 {bone1Name=dyn_jacket_02_R_03, bone2Name=dyn_jacket_02_R_02, swing1LimitDegrees=10, swing2LimitDegrees=25, twistLimitDegrees=2}
 {bone1Name=dyn_jacket_03_R_04, bone2Name=dyn_jacket_03_R_03, swing1LimitDegrees=15, swing2LimitDegrees=30, twistLimitDegrees=2}
 {
  bone1Name=dyn_jacket_03_R_mstr
  bone2Name=spine_01
  swing1LimitDegrees=45
  swing1Motion=Free
  swing2LimitDegrees=45
  swing2Motion=Free
  twistLimitDegrees=45
  twistMotion=Free
 }
 {bone1Name=dyn_jacket_03_R_01, bone2Name=dyn_jacket_03_R_mstr, swing1LimitDegrees=10, swing2LimitDegrees=20, twistLimitDegrees=2}
 {bone1Name=dyn_jacket_03_R_02, bone2Name=dyn_jacket_03_R_01, swing1LimitDegrees=10, swing2LimitDegrees=20, twistLimitDegrees=2}
 {bone1Name=dyn_jacket_03_R_03, bone2Name=dyn_jacket_03_R_02, swing1LimitDegrees=10, swing2LimitDegrees=25, twistLimitDegrees=2}
 {bone1Name=dyn_jacket_B_R, bone2Name=dyn_jacket_B_02, swing1LimitDegrees=30, swing2LimitDegrees=30, twistLimitDegrees=2}
]
