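# Animation Blueprint for the LazarusLens character body, adding secondary motion to scarf, necklace, jacket and cloth folds.
# The copied parent pose passes through constraints, Control Rig and AnimDynamics into a cache feeding default, skydive and wings RigidBody branches.
# IsWearingWings, set when the matching LazarusLens pack is detected, selects the wings branch.
# The event graph also updates gravity override, crouch alpha and material WPO cloth parameters, including skydive wind.
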AnimBlueprint /BRCosmetics/Characters/Player/Male/Medium/Bodies/M_MED_LazarusLens/Meshes/M_MED_LazarusLens_AnimBP
target_skeleton=&/M_MED_LazarusLens_Skeleton
graphs [
 AnimGraph {nodes [
   n1 = "Constraint - dyn_scarf_L_mstr (PP)"!(ComponentPose=Local To Component(LocalPose=Use cached pose 'CtrlRigCache'()), Alpha=1, ConstraintWeights_0=1, ConstraintWeights_1:float=0) -> (Pose) {
    alphaBoolBlend {blendOption=Linear}
    boneToModify {boneName=dyn_scarf_L_mstr}
    constraintSetup [{targetBone {boneName=spine_05}}, {targetBone {boneName=clavicle_l}}]
    constraintWeights [1,1]
   }
   n2 = Anim Dynamics - Bone: dyn_scarf_L_01(ComponentPose=@n1, Alpha=1, ExternalForce="0,0,0") {
    angularDampingOverride=0.9999
    angularSpringConstant=500
    bAngularSpring=true
    bOverrideAngularDamping=true
    simulationSpace=RootRelative
    alphaBoolBlend {blendOption=Linear}
    boundBone {boneName=dyn_scarf_L_01}
    componentAppliedLinearAccClamp=<1000 1000 1000>
    componentLinearAccScale=<0.1 0.1 0.2>
    componentLinearVelScale=<0.1 0.1 0.1>
    physicsBodyDefinitions [{
      boundBone {boneName=dyn_scarf_L_01}
      boxExtents=<50 50 50>
      constraintSetup {angularLimitsMax=<1 0 5>, angularLimitsMin=<-1 -20 -5>, angularTarget=<0 0 0>}
      localJointOffset=<5 0 0>
     }]
    planarLimits [{drivingBone {boneName=spine_04}, planeTransform {location=<5 11 0>, rotation=<rot 0 0 90>}}]
    relativeSpaceBone {boneName=spine_05}
   }
   n3 = Anim Dynamics - Bone: dyn_scarf_L_02(ComponentPose=@n2, Alpha=1, ExternalForce="0,0,0") {
    angularDampingOverride=0.9999
    angularSpringConstant=500
    bAngularSpring=true
    bOverrideAngularDamping=true
    simulationSpace=RootRelative
    alphaBoolBlend {blendOption=Linear}
    boundBone {boneName=dyn_scarf_L_02}
    componentAppliedLinearAccClamp=<1000 1000 1000>
    componentLinearAccScale=<0.1 0.1 0.2>
    componentLinearVelScale=<0.1 0.1 0.1>
    physicsBodyDefinitions [{
      boundBone {boneName=dyn_scarf_L_02}
      boxExtents=<30 30 30>
      constraintSetup {angularLimitsMax=<1 0 10>, angularLimitsMin=<-1 -20 -10>, angularTarget=<0 0 0>}
      localJointOffset=<5 0 0>
     }]
    planarLimits [{drivingBone {boneName=spine_04}, planeTransform {location=<5 12.5 0>, rotation=<rot 0 0 90>}}]
    relativeSpaceBone {boneName=spine_05}
   }
   n4 = Anim Dynamics - Bone: dyn_scarf_L_03(ComponentPose=@n3, Alpha=1, ExternalForce="0,0,0") {
    angularDampingOverride=0.9999
    angularSpringConstant=400
    bAngularSpring=true
    bOverrideAngularDamping=true
    simulationSpace=RootRelative
    alphaBoolBlend {blendOption=Linear}
    boundBone {boneName=dyn_scarf_L_03}
    componentAppliedLinearAccClamp=<1000 1000 1000>
    componentLinearAccScale=<0.1 0.1 0.2>
    componentLinearVelScale=<0.1 0.1 0.1>
    physicsBodyDefinitions [{
      boundBone {boneName=dyn_scarf_L_03}
      boxExtents=<25 25 25>
      constraintSetup {angularLimitsMax=<1 5 20>, angularLimitsMin=<-1 -20 -20>, angularTarget=<0 0 0>}
      localJointOffset=<5 0 0>
     }]
    planarLimits [
     {drivingBone {boneName=spine_04}, planeTransform {location=<5 12.5 0>, rotation=<rot 0 0 90>, scale=<0.01 0.01 0.01>}}
     {drivingBone {boneName=lowerarm_l}, planeTransform {location=<0 10 0>, rotation=<rot 0 0 119.9999>}}
    ]
    relativeSpaceBone {boneName=spine_05}
   }
   n5 = Anim Dynamics - Bone: dyn_scarf_L_04(ComponentPose=@n4, Alpha=1, ExternalForce="0,0,0") {
    angularDampingOverride=0.9999
    angularSpringConstant=300
    bAngularSpring=true
    bOverrideAngularDamping=true
    simulationSpace=RootRelative
    alphaBoolBlend {blendOption=Linear}
    boundBone {boneName=dyn_scarf_L_04}
    componentAppliedLinearAccClamp=<1000 1000 1000>
    componentLinearAccScale=<0.1 0.1 0.2>
    componentLinearVelScale=<0.1 0.1 0.1>
    physicsBodyDefinitions [{
      boundBone {boneName=dyn_scarf_L_04}
      boxExtents=<20 20 20>
      constraintSetup {angularLimitsMax=<1 5 25>, angularLimitsMin=<-1 -25 -25>, angularTarget=<0 0 0>}
      localJointOffset=<5 0 0>
     }]
    planarLimits [
     {drivingBone {boneName=spine_04}, planeTransform {location=<5 12.5 0>, rotation=<rot 0 0 90>}}
     {drivingBone {boneName=lowerarm_l}, planeTransform {location=<0 5 0>, rotation=<rot 0 0 119.9999>}}
    ]
    relativeSpaceBone {boneName=spine_05}
   }
   n6 = "Constraint - dyn_scarf_R_mstr (PP)"!(ComponentPose=@n5, Alpha=1, ConstraintWeights_0=1, ConstraintWeights_1:float=0) -> (Pose) {
    alphaBoolBlend {blendOption=Linear}
    boneToModify {boneName=dyn_scarf_R_mstr}
    constraintSetup [{targetBone {boneName=spine_05}}, {targetBone {boneName=clavicle_r}}]
    constraintWeights [1,1]
   }
   n7 = Anim Dynamics - Bone: dyn_scarf_R_01(ComponentPose=@n6, Alpha=1, ExternalForce="0,0,0") {
    angularDampingOverride=0.9999
    angularSpringConstant=500
    bAngularSpring=true
    bOverrideAngularDamping=true
    simulationSpace=RootRelative
    alphaBoolBlend {blendOption=Linear}
    boundBone {boneName=dyn_scarf_R_01}
    componentAppliedLinearAccClamp=<1000 1000 1000>
    componentLinearAccScale=<0.1 0.1 0.2>
    componentLinearVelScale=<0.1 0.1 0.1>
    physicsBodyDefinitions [{
      boundBone {boneName=dyn_scarf_R_01}
      boxExtents=<50 50 50>
      constraintSetup {angularLimitsMax=<1 0 5>, angularLimitsMin=<-1 -20 -5>, angularTarget=<0 0 0>}
      localJointOffset=<5 0 0>
     }]
    planarLimits [{drivingBone {boneName=spine_04}, planeTransform {location=<5 10 10>, rotation=<rot 0 -12 90>}}]
    relativeSpaceBone {boneName=spine_05}
   }
   n8 = Anim Dynamics - Bone: dyn_scarf_R_02(ComponentPose=@n7, Alpha=1, ExternalForce="0,0,0") {
    angularDampingOverride=0.9999
    angularSpringConstant=500
    bAngularSpring=true
    bOverrideAngularDamping=true
    simulationSpace=RootRelative
    alphaBoolBlend {blendOption=Linear}
    boundBone {boneName=dyn_scarf_R_02}
    componentAppliedLinearAccClamp=<1000 1000 1000>
    componentLinearAccScale=<0.1 0.1 0.2>
    componentLinearVelScale=<0.1 0.1 0.1>
    physicsBodyDefinitions [{
      boundBone {boneName=dyn_scarf_R_02}
      boxExtents=<30 30 30>
      constraintSetup {angularLimitsMax=<1 0 10>, angularLimitsMin=<-1 -20 -10>, angularTarget=<0 0 0>}
      localJointOffset=<5 0 0>
     }]
    planarLimits [{drivingBone {boneName=spine_04}, planeTransform {location=<5 13 10>, rotation=<rot 0 0 94.9999>}}]
    relativeSpaceBone {boneName=spine_05}
   }
   n9 = Anim Dynamics - Bone: dyn_scarf_R_03(ComponentPose=@n8, Alpha=1, ExternalForce="0,0,0") {
    angularDampingOverride=0.9999
    angularSpringConstant=400
    bAngularSpring=true
    bOverrideAngularDamping=true
    simulationSpace=RootRelative
    alphaBoolBlend {blendOption=Linear}
    boundBone {boneName=dyn_scarf_R_03}
    componentAppliedLinearAccClamp=<1000 1000 1000>
    componentLinearAccScale=<0.1 0.1 0.2>
    componentLinearVelScale=<0.1 0.1 0.1>
    physicsBodyDefinitions [{
      boundBone {boneName=dyn_scarf_R_03}
      boxExtents=<25 25 25>
      constraintSetup {angularLimitsMax=<1 5 20>, angularLimitsMin=<-1 -20 -20>, angularTarget=<0 0 0>}
      localJointOffset=<5 0 0>
     }]
    planarLimits [{drivingBone {boneName=spine_04}, planeTransform {location=<5 13.5 10>, rotation=<rot 0 0 94.9999>}}]
    relativeSpaceBone {boneName=spine_05}
   }
   n10 = Anim Dynamics - Bone: dyn_scarf_R_04(ComponentPose=@n9, Alpha=1, ExternalForce="0,0,0") {
    angularDampingOverride=0.9999
    angularSpringConstant=300
    bAngularSpring=true
    bOverrideAngularDamping=true
    simulationSpace=RootRelative
    alphaBoolBlend {blendOption=Linear}
    boundBone {boneName=dyn_scarf_R_04}
    componentAppliedLinearAccClamp=<1000 1000 1000>
    componentLinearAccScale=<0.1 0.1 0.2>
    componentLinearVelScale=<0.1 0.1 0.1>
    physicsBodyDefinitions [{
      boundBone {boneName=dyn_scarf_R_04}
      boxExtents=<20 20 20>
      constraintSetup {angularLimitsMax=<1 5 25>, angularLimitsMin=<-1 -25 -25>, angularTarget=<0 0 0>}
      localJointOffset=<5 0 0>
     }]
    planarLimits [{drivingBone {boneName=spine_04}, planeTransform {location=<5 13.5 10>, rotation=<rot 0 0 94.9999>}}]
    relativeSpaceBone {boneName=spine_05}
   }
   n11 = RigidBody!(ComponentPose=@n10, Alpha, OverrideWorldGravity:Vector="0,0,-920", ExternalForce) -> (Pose) {
    bClampLinearTranslationLimitToRefPose=true
    bOverrideWorldGravity=true
    overridePhysicsAsset=&/M_MED_LazarusLens_Physics
    simulationSpace=BaseBoneSpace
    alphaBoolBlend {blendOption=Linear}
    baseBoneRef {boneName=pelvis}
    componentAppliedLinearAccClamp=<3000 3000 1000>
    componentLinearAccScale=<0.4 0.4 0.1>
    componentLinearVelScale=<0.4 0.4 0.1>
    overrideWorldGravity {z=-920}
    simSpaceSettings {velocityScaleZ=0.5, worldAlpha=0.3}
   }
   n12 = Component To Local(ComponentPose=@n11)
   n13 = "Constraint - dyn_scarf_L_mstr (PP)"(ComponentPose=Local To Component(LocalPose=Use cached pose 'CtrlRigCache'()), Alpha=1, ConstraintWeights_0=1, ConstraintWeights_1=0) {
    alphaBoolBlend {blendOption=Linear}
    boneToModify {boneName=dyn_scarf_L_mstr}
    constraintSetup [{targetBone {boneName=spine_05}}, {targetBone {boneName=clavicle_l}}]
    constraintWeights [1,1]
   }
   n14 = Anim Dynamics - Bone: dyn_scarf_L_01!(ComponentPose=@n13, Alpha=1, GravityOverride:Vector="0,0,0", ExternalForce="0,0,0") -> (Pose) {
    angularDampingOverride=0.9999
    angularSpringConstant=500
    bAngularSpring=true
    bOverrideAngularDamping=true
    bUseGravityOverride=true
    simulationSpace=RootRelative
    alphaBoolBlend {blendOption=Linear}
    boundBone {boneName=dyn_scarf_L_01}
    componentAppliedLinearAccClamp=<1000 1000 1000>
    componentLinearAccScale=<0.1 0.1 0.2>
    componentLinearVelScale=<0.1 0.1 0.1>
    physicsBodyDefinitions [{
      boundBone {boneName=dyn_scarf_L_01}
      boxExtents=<50 50 50>
      constraintSetup {angularLimitsMax=<1 0 5>, angularLimitsMin=<-1 -20 -5>, angularTarget=<0 0 0>}
      localJointOffset=<5 0 0>
     }]
    planarLimits [{drivingBone {boneName=spine_04}, planeTransform {location=<5 11 0>, rotation=<rot 0 0 90>}}]
    relativeSpaceBone {boneName=spine_05}
   }
   n15 = Anim Dynamics - Bone: dyn_scarf_L_02!(ComponentPose=@n14, Alpha=1, GravityOverride:Vector="0,0,0", ExternalForce="0,0,0") -> (Pose) {
    angularDampingOverride=0.9999
    angularSpringConstant=500
    bAngularSpring=true
    bOverrideAngularDamping=true
    bUseGravityOverride=true
    simulationSpace=RootRelative
    alphaBoolBlend {blendOption=Linear}
    boundBone {boneName=dyn_scarf_L_02}
    componentAppliedLinearAccClamp=<1000 1000 1000>
    componentLinearAccScale=<0.1 0.1 0.2>
    componentLinearVelScale=<0.1 0.1 0.1>
    physicsBodyDefinitions [{
      boundBone {boneName=dyn_scarf_L_02}
      boxExtents=<30 30 30>
      constraintSetup {angularLimitsMax=<1 0 10>, angularLimitsMin=<-1 -20 -10>, angularTarget=<0 0 0>}
      localJointOffset=<5 0 0>
     }]
    planarLimits [{drivingBone {boneName=spine_04}, planeTransform {location=<5 12.5 0>, rotation=<rot 0 0 90>}}]
    relativeSpaceBone {boneName=spine_05}
   }
   n16 = Anim Dynamics - Bone: dyn_scarf_L_03!(ComponentPose=@n15, Alpha=1, GravityOverride:Vector="0,0,0", ExternalForce="0,0,0") -> (Pose) {
    angularDampingOverride=0.9999
    angularSpringConstant=400
    bAngularSpring=true
    bOverrideAngularDamping=true
    bUseGravityOverride=true
    simulationSpace=RootRelative
    alphaBoolBlend {blendOption=Linear}
    boundBone {boneName=dyn_scarf_L_03}
    componentAppliedLinearAccClamp=<1000 1000 1000>
    componentLinearAccScale=<0.1 0.1 0.2>
    componentLinearVelScale=<0.1 0.1 0.1>
    physicsBodyDefinitions [{
      boundBone {boneName=dyn_scarf_L_03}
      boxExtents=<25 25 25>
      constraintSetup {angularLimitsMax=<1 5 20>, angularLimitsMin=<-1 -20 -20>, angularTarget=<0 0 0>}
      localJointOffset=<5 0 0>
     }]
    planarLimits [
     {drivingBone {boneName=spine_04}, planeTransform {location=<5 12.5 0>, rotation=<rot 0 0 90>, scale=<0.01 0.01 0.01>}}
     {drivingBone {boneName=lowerarm_l}, planeTransform {location=<0 10 0>, rotation=<rot 0 0 119.9999>}}
    ]
    relativeSpaceBone {boneName=spine_05}
   }
   n17 = Anim Dynamics - Bone: dyn_scarf_L_04!(ComponentPose=@n16, Alpha=1, GravityOverride:Vector="0,0,0", ExternalForce="0,0,0") -> (Pose) {
    angularDampingOverride=0.9999
    angularSpringConstant=300
    bAngularSpring=true
    bOverrideAngularDamping=true
    bUseGravityOverride=true
    simulationSpace=RootRelative
    alphaBoolBlend {blendOption=Linear}
    boundBone {boneName=dyn_scarf_L_04}
    componentAppliedLinearAccClamp=<1000 1000 1000>
    componentLinearAccScale=<0.1 0.1 0.2>
    componentLinearVelScale=<0.1 0.1 0.1>
    physicsBodyDefinitions [{
      boundBone {boneName=dyn_scarf_L_04}
      boxExtents=<20 20 20>
      constraintSetup {angularLimitsMax=<1 5 25>, angularLimitsMin=<-1 -25 -25>, angularTarget=<0 0 0>}
      localJointOffset=<5 0 0>
     }]
    planarLimits [
     {drivingBone {boneName=spine_04}, planeTransform {location=<5 12.5 0>, rotation=<rot 0 0 90>}}
     {drivingBone {boneName=lowerarm_l}, planeTransform {location=<0 5 0>, rotation=<rot 0 0 119.9999>}}
    ]
    relativeSpaceBone {boneName=spine_05}
   }
   n18 = "Constraint - dyn_scarf_R_mstr (PP)"(ComponentPose=@n17, Alpha=1, ConstraintWeights_0=1, ConstraintWeights_1=0) {
    alphaBoolBlend {blendOption=Linear}
    boneToModify {boneName=dyn_scarf_R_mstr}
    constraintSetup [{targetBone {boneName=spine_05}}, {targetBone {boneName=clavicle_r}}]
    constraintWeights [1,1]
   }
   n19 = Anim Dynamics - Bone: dyn_scarf_R_01!(ComponentPose=@n18, Alpha=1, GravityOverride:Vector="0,0,0", ExternalForce="0,0,0") -> (Pose) {
    angularDampingOverride=0.9999
    angularSpringConstant=500
    bAngularSpring=true
    bOverrideAngularDamping=true
    bUseGravityOverride=true
    simulationSpace=RootRelative
    alphaBoolBlend {blendOption=Linear}
    boundBone {boneName=dyn_scarf_R_01}
    componentAppliedLinearAccClamp=<1000 1000 1000>
    componentLinearAccScale=<0.1 0.1 0.2>
    componentLinearVelScale=<0.1 0.1 0.1>
    physicsBodyDefinitions [{
      boundBone {boneName=dyn_scarf_R_01}
      boxExtents=<50 50 50>
      constraintSetup {angularLimitsMax=<1 0 5>, angularLimitsMin=<-1 -20 -5>, angularTarget=<0 0 0>}
      localJointOffset=<5 0 0>
     }]
    planarLimits [{drivingBone {boneName=spine_04}, planeTransform {location=<5 16 10>, rotation=<rot 0 -12 90>}}]
    relativeSpaceBone {boneName=spine_05}
   }
   n20 = Anim Dynamics - Bone: dyn_scarf_R_02!(ComponentPose=@n19, Alpha=1, GravityOverride:Vector="0,0,0", ExternalForce="0,0,0") -> (Pose) {
    angularDampingOverride=0.9999
    angularSpringConstant=500
    bAngularSpring=true
    bOverrideAngularDamping=true
    bUseGravityOverride=true
    simulationSpace=RootRelative
    alphaBoolBlend {blendOption=Linear}
    boundBone {boneName=dyn_scarf_R_02}
    componentAppliedLinearAccClamp=<1000 1000 1000>
    componentLinearAccScale=<0.1 0.1 0.2>
    componentLinearVelScale=<0.1 0.1 0.1>
    physicsBodyDefinitions [{
      boundBone {boneName=dyn_scarf_R_02}
      boxExtents=<30 30 30>
      constraintSetup {angularLimitsMax=<1 0 10>, angularLimitsMin=<-1 -20 -10>, angularTarget=<0 0 0>}
      localJointOffset=<5 0 0>
     }]
    planarLimits [{drivingBone {boneName=spine_04}, planeTransform {location=<5 14 10>, rotation=<rot 0 0 94.9999>}}]
    relativeSpaceBone {boneName=spine_05}
   }
   n21 = Anim Dynamics - Bone: dyn_scarf_R_03!(ComponentPose=@n20, Alpha=1, GravityOverride:Vector="0,0,0", ExternalForce="0,0,0") -> (Pose) {
    angularDampingOverride=0.9999
    angularSpringConstant=400
    bAngularSpring=true
    bOverrideAngularDamping=true
    bUseGravityOverride=true
    simulationSpace=RootRelative
    alphaBoolBlend {blendOption=Linear}
    boundBone {boneName=dyn_scarf_R_03}
    componentAppliedLinearAccClamp=<1000 1000 1000>
    componentLinearAccScale=<0.1 0.1 0.2>
    componentLinearVelScale=<0.1 0.1 0.1>
    physicsBodyDefinitions [{
      boundBone {boneName=dyn_scarf_R_03}
      boxExtents=<25 25 25>
      constraintSetup {angularLimitsMax=<1 5 20>, angularLimitsMin=<-1 -20 -20>, angularTarget=<0 0 0>}
      localJointOffset=<5 0 0>
     }]
    planarLimits [{drivingBone {boneName=spine_04}, planeTransform {location=<5 14 10>, rotation=<rot 0 0 94.9999>}}]
    relativeSpaceBone {boneName=spine_05}
   }
   n22 = Anim Dynamics - Bone: dyn_scarf_R_04!(ComponentPose=@n21, Alpha=1, GravityOverride:Vector="0,0,0", ExternalForce="0,0,0") -> (Pose) {
    angularDampingOverride=0.9999
    angularSpringConstant=300
    bAngularSpring=true
    bOverrideAngularDamping=true
    bUseGravityOverride=true
    simulationSpace=RootRelative
    alphaBoolBlend {blendOption=Linear}
    boundBone {boneName=dyn_scarf_R_04}
    componentAppliedLinearAccClamp=<1000 1000 1000>
    componentLinearAccScale=<0.1 0.1 0.2>
    componentLinearVelScale=<0.1 0.1 0.1>
    physicsBodyDefinitions [{
      boundBone {boneName=dyn_scarf_R_04}
      boxExtents=<20 20 20>
      constraintSetup {angularLimitsMax=<1 5 25>, angularLimitsMin=<-1 -25 -25>, angularTarget=<0 0 0>}
      localJointOffset=<5 0 0>
     }]
    planarLimits [{drivingBone {boneName=spine_04}, planeTransform {location=<5 14 10>, rotation=<rot 0 0 94.9999>}}]
    relativeSpaceBone {boneName=spine_05}
   }
   n23 = RigidBody(ComponentPose=@n22, OverrideWorldGravity="0,0,-920") {
    bClampLinearTranslationLimitToRefPose=true
    bOverrideWorldGravity=true
    overridePhysicsAsset=&/M_MED_LazarusLens_Physics_Skydive
    simulationSpace=BaseBoneSpace
    alphaBoolBlend {blendOption=Linear}
    baseBoneRef {boneName=pelvis}
    componentAppliedLinearAccClamp=<3000 3000 1000>
    componentLinearAccScale=<0.4 0.4 0.1>
    componentLinearVelScale=<0.4 0.4 0.1>
    overrideWorldGravity {z=-920}
    simSpaceSettings {velocityScaleZ=0.5, worldAlpha=0.3}
   }
   n24 = Component To Local(ComponentPose=@n23)
   n25 = "Constraint - dyn_button (PP)"!(
    ComponentPose=Local To Component(LocalPose=Blend Poses by bool(BlendPose_0=@n24, BlendPose_1=@n12) {blendType=Linear})
    Alpha=1
    ConstraintWeights_0=0.45
    ConstraintWeights_1:float=0.55
   ) -> (Pose) {
    alphaBoolBlend {blendOption=Linear}
    boneToModify {boneName=dyn_button}
    constraintSetup [{targetBone {boneName=dyn_jacket_01_L_01}}, {targetBone {boneName=dyn_jacket_01_L_02}}]
    constraintWeights [1,1]
   }
   n26 = RigidBody(ComponentPose=Local To Component(LocalPose=Use cached pose 'CtrlRigCache'()), OverrideWorldGravity="0,0,-920") {
    bClampLinearTranslationLimitToRefPose=true
    bOverrideWorldGravity=true
    overridePhysicsAsset=&/M_MED_LazarusLens_Physics_Wings
    simulationSpace=BaseBoneSpace
    alphaBoolBlend {blendOption=Linear}
    baseBoneRef {boneName=pelvis}
    componentAppliedLinearAccClamp=<3000 3000 1000>
    componentLinearAccScale=<0.4 0.4 0.1>
    componentLinearVelScale=<0.4 0.4 0.1>
    overrideWorldGravity {z=-920}
    simSpaceSettings {velocityScaleZ=0.5, worldAlpha=0.3}
   }
   Output Pose(Result=Blend Poses by bool(
     BlendPose_0=Component To Local(ComponentPose=@n26)
     BlendPose_1=Component To Local(ComponentPose=@n25)
     bActiveValue=Get IsWearingWings() -> (IsWearingWings:bool)
    ) {blendType=Linear}) {layerGroup=None}
   n27 = "Constraint - dyn_knotcenter_mstr (PPPP)"!(
    ComponentPose=Local To Component(LocalPose=Copy Pose From Mesh() {bUseAttachedParent=true})
    Alpha=1
    ConstraintWeights_0=0.4694
    ConstraintWeights_1:float=0.3925
    ConstraintWeights_2:float=0.115
    ConstraintWeights_3:float=0.0231
   ) -> (Pose) {
    alphaBoolBlend {blendOption=Linear}
    boneToModify {boneName=dyn_knotcenter_mstr}
    constraintSetup [{targetBone {boneName=pelvis}}, {targetBone {boneName=spine_01}}, {targetBone {boneName=spine_02}}, {targetBone {boneName=spine_03}}]
    constraintWeights [1,1,1,1]
   }
   n28 = Control Rig!(Source=Component To Local(ComponentPose=@n27), Alpha, IsCrouchCtrl:double=0, IsCrouchBool:bool=false, RbanOffCtrl:bool=false) -> (Pose) {
    alphaBoolBlend {blendOption=Linear}
    controlRigAssetReference {blueprintRigClass=&/M_MED_LazarusLens_CtrlRig.M_MED_LazarusLens_CtrlRig_C}
   }
   n29 = Anim Dynamics - Bone: dyn_necklace_01(ComponentPose=Local To Component(LocalPose=@n28), Alpha=1, ExternalForce="0,0,0") {
    angularDampingOverride=0.8
    angularSpringConstant=300
    bAngularSpring=true
    bUseSphericalLimits=true
    simulationSpace=BoneRelative
    alphaBoolBlend {blendOption=Linear}
    boundBone {boneName=dyn_necklace_01}
    componentAppliedLinearAccClamp=<10000 10000 10000>
    componentLinearAccScale=<1 1 1>
    componentLinearVelScale=<1 1 1>
    physicsBodyDefinitions [{
      boundBone {boneName=dyn_necklace_01}
      boxExtents=<30 30 30>
      constraintSetup {angularLimitsMax=<0 5 20>, angularLimitsMin=<0 -5 0>, angularTarget=<0 0 0>}
      localJointOffset=<1 0 0>
     }]
    planarLimits [{drivingBone {boneName=spine_05}, planeTransform {location=<20 0 5>, rotation=<rot 45 44.9999 179.9999>}}]
    relativeSpaceBone {boneName=root}
    sphericalLimits [{limitRadius=3, drivingBone {boneName=spine_05}, sphereLocalOffset=<7 4 2>}]
   }
   n30 = Anim Dynamics - Bone: dyn_necklace_02(ComponentPose=@n29, Alpha=1, ExternalForce="0,0,0") {
    angularDampingOverride=0.999
    bOverrideAngularDamping=true
    bUsePlanarLimit=false
    bUseSphericalLimits=true
    simulationSpace=RootRelative
    alphaBoolBlend {blendOption=Linear}
    boundBone {boneName=dyn_necklace_02}
    componentAppliedLinearAccClamp=<10000 10000 10000>
    componentLinearAccScale=<1 1 1>
    componentLinearVelScale=<1 1 1>
    physicsBodyDefinitions [{
      boundBone {boneName=dyn_necklace_02}
      boxExtents=<10 10 10>
      constraintSetup {angularLimitsMax=<0 20 15>, angularLimitsMin=<0 -20 0>, angularTarget=<0 0 0>}
      localJointOffset=<0.5 0 0>
     }]
    relativeSpaceBone {boneName=pelvis}
    sphericalLimits [{limitRadius=3.5, drivingBone {boneName=spine_05}, sphereLocalOffset=<0 10 3>}]
   }
   n31 = "Constraint - dyn_scarf_L_mstr (PP)"(ComponentPose=@n30, Alpha=1, ConstraintWeights_0=1, ConstraintWeights_1=0) {
    alphaBoolBlend {blendOption=Linear}
    boneToModify {boneName=dyn_scarf_L_mstr}
    constraintSetup [{targetBone {boneName=spine_05}}, {targetBone {boneName=clavicle_l}}]
    constraintWeights [1,1]
   }
   n32 = Anim Dynamics - Bone: dyn_scarf_L_01!(ComponentPose=@n31, Alpha=1, ExternalForce="0,0,0") -> (Pose) {
    angularDampingOverride=0.9999
    angularSpringConstant=500
    bAngularSpring=true
    bOverrideAngularDamping=true
    simulationSpace=RootRelative
    alphaBoolBlend {blendOption=Linear}
    boundBone {boneName=dyn_scarf_L_01}
    componentAppliedLinearAccClamp=<1000 1000 1000>
    componentLinearAccScale=<0.1 0.1 0.2>
    componentLinearVelScale=<0.1 0.1 0.1>
    physicsBodyDefinitions [{
      boundBone {boneName=dyn_scarf_L_01}
      boxExtents=<50 50 50>
      constraintSetup {angularLimitsMax=<1 0 5>, angularLimitsMin=<-1 -20 -5>, angularTarget=<0 0 0>}
      localJointOffset=<5 0 0>
     }]
    planarLimits [{drivingBone {boneName=spine_04}, planeTransform {location=<5 11 0>, rotation=<rot 0 0 90>}}]
    relativeSpaceBone {boneName=spine_05}
   }
   n33 = Anim Dynamics - Bone: dyn_scarf_L_02!(ComponentPose=@n32, Alpha=1, ExternalForce="0,0,0") -> (Pose) {
    angularDampingOverride=0.9999
    angularSpringConstant=500
    bAngularSpring=true
    bOverrideAngularDamping=true
    simulationSpace=RootRelative
    alphaBoolBlend {blendOption=Linear}
    boundBone {boneName=dyn_scarf_L_02}
    componentAppliedLinearAccClamp=<1000 1000 1000>
    componentLinearAccScale=<0.1 0.1 0.2>
    componentLinearVelScale=<0.1 0.1 0.1>
    physicsBodyDefinitions [{
      boundBone {boneName=dyn_scarf_L_02}
      boxExtents=<30 30 30>
      constraintSetup {angularLimitsMax=<1 0 10>, angularLimitsMin=<-1 -20 -10>, angularTarget=<0 0 0>}
      localJointOffset=<5 0 0>
     }]
    planarLimits [{drivingBone {boneName=spine_04}, planeTransform {location=<5 12.5 0>, rotation=<rot 0 0 90>}}]
    relativeSpaceBone {boneName=spine_05}
   }
   n34 = Anim Dynamics - Bone: dyn_scarf_L_03!(ComponentPose=@n33, Alpha=0, ExternalForce="0,0,0") -> (Pose) {
    angularDampingOverride=0.9999
    angularSpringConstant=400
    bAngularSpring=true
    bOverrideAngularDamping=true
    simulationSpace=RootRelative
    alphaBoolBlend {blendOption=Linear}
    boundBone {boneName=dyn_scarf_L_03}
    componentAppliedLinearAccClamp=<1000 1000 1000>
    componentLinearAccScale=<0.1 0.1 0.2>
    componentLinearVelScale=<0.1 0.1 0.1>
    physicsBodyDefinitions [{
      boundBone {boneName=dyn_scarf_L_03}
      boxExtents=<25 25 25>
      constraintSetup {angularLimitsMax=<1 5 20>, angularLimitsMin=<-1 -20 -20>, angularTarget=<0 0 0>}
      localJointOffset=<5 0 0>
     }]
    planarLimits [
     {drivingBone {boneName=spine_04}, planeTransform {location=<5 12.5 0>, rotation=<rot 0 0 90>, scale=<0.01 0.01 0.01>}}
     {drivingBone {boneName=lowerarm_l}, planeTransform {location=<0 10 0>, rotation=<rot 0 0 120>}}
    ]
    relativeSpaceBone {boneName=spine_05}
   }
   n35 = Anim Dynamics - Bone: dyn_scarf_L_04!(ComponentPose=@n34, Alpha=0, ExternalForce="0,0,0") -> (Pose) {
    angularDampingOverride=0.9999
    angularSpringConstant=300
    bAngularSpring=true
    bOverrideAngularDamping=true
    simulationSpace=RootRelative
    alphaBoolBlend {blendOption=Linear}
    boundBone {boneName=dyn_scarf_L_04}
    componentAppliedLinearAccClamp=<1000 1000 1000>
    componentLinearAccScale=<0.1 0.1 0.2>
    componentLinearVelScale=<0.1 0.1 0.1>
    physicsBodyDefinitions [{
      boundBone {boneName=dyn_scarf_L_04}
      boxExtents=<20 20 20>
      constraintSetup {angularLimitsMax=<1 5 25>, angularLimitsMin=<-1 -25 -25>, angularTarget=<0 0 0>}
      localJointOffset=<5 0 0>
     }]
    planarLimits [
     {drivingBone {boneName=spine_04}, planeTransform {location=<5 12.5 0>, rotation=<rot 0 0 90>}}
     {drivingBone {boneName=lowerarm_l}, planeTransform {location=<0 5 0>, rotation=<rot 0 0 119.9999>}}
    ]
    relativeSpaceBone {boneName=spine_05}
   }
   n36 = "Constraint - dyn_scarf_R_mstr (PP)"(ComponentPose=@n35, Alpha=1, ConstraintWeights_0=1, ConstraintWeights_1=0) {
    alphaBoolBlend {blendOption=Linear}
    boneToModify {boneName=dyn_scarf_R_mstr}
    constraintSetup [{targetBone {boneName=spine_05}}, {targetBone {boneName=clavicle_r}}]
    constraintWeights [1,1]
   }
   n37 = Anim Dynamics - Bone: dyn_scarf_R_01(ComponentPose=@n36, Alpha=1, GravityOverride="0,0,0", ExternalForce="0,0,0") {
    angularDampingOverride=0.9999
    angularSpringConstant=500
    bAngularSpring=true
    bOverrideAngularDamping=true
    bUseGravityOverride=true
    simulationSpace=RootRelative
    alphaBoolBlend {blendOption=Linear}
    boundBone {boneName=dyn_scarf_R_01}
    componentAppliedLinearAccClamp=<1000 1000 1000>
    componentLinearAccScale=<0.1 0.1 0.2>
    componentLinearVelScale=<0.1 0.1 0.1>
    physicsBodyDefinitions [{
      boundBone {boneName=dyn_scarf_R_01}
      boxExtents=<50 50 50>
      constraintSetup {angularLimitsMax=<1 0 5>, angularLimitsMin=<-1 -20 -5>, angularTarget=<0 0 0>}
      localJointOffset=<5 0 0>
     }]
    planarLimits [{drivingBone {boneName=spine_04}, planeTransform {location=<5 10 10>, rotation=<rot 0 -12 90>}}]
    relativeSpaceBone {boneName=spine_05}
   }
   n38 = Anim Dynamics - Bone: dyn_scarf_R_02(ComponentPose=@n37, Alpha=1, GravityOverride="0,0,0", ExternalForce="0,0,0") {
    angularDampingOverride=0.9999
    angularSpringConstant=500
    bAngularSpring=true
    bOverrideAngularDamping=true
    bUseGravityOverride=true
    simulationSpace=RootRelative
    alphaBoolBlend {blendOption=Linear}
    boundBone {boneName=dyn_scarf_R_02}
    componentAppliedLinearAccClamp=<1000 1000 1000>
    componentLinearAccScale=<0.1 0.1 0.2>
    componentLinearVelScale=<0.1 0.1 0.1>
    physicsBodyDefinitions [{
      boundBone {boneName=dyn_scarf_R_02}
      boxExtents=<30 30 30>
      constraintSetup {angularLimitsMax=<1 0 10>, angularLimitsMin=<-1 -20 -10>, angularTarget=<0 0 0>}
      localJointOffset=<5 0 0>
     }]
    planarLimits [{drivingBone {boneName=spine_04}, planeTransform {location=<5 13 10>, rotation=<rot 0 0 95>}}]
    relativeSpaceBone {boneName=spine_05}
   }
   n39 = Anim Dynamics - Bone: dyn_scarf_R_03(ComponentPose=@n38, Alpha=1, GravityOverride="0,0,0", ExternalForce="0,0,0") {
    angularDampingOverride=0.9999
    angularSpringConstant=400
    bAngularSpring=true
    bOverrideAngularDamping=true
    bUseGravityOverride=true
    simulationSpace=RootRelative
    alphaBoolBlend {blendOption=Linear}
    boundBone {boneName=dyn_scarf_R_03}
    componentAppliedLinearAccClamp=<1000 1000 1000>
    componentLinearAccScale=<0.1 0.1 0.2>
    componentLinearVelScale=<0.1 0.1 0.1>
    physicsBodyDefinitions [{
      boundBone {boneName=dyn_scarf_R_03}
      boxExtents=<25 25 25>
      constraintSetup {angularLimitsMax=<1 5 20>, angularLimitsMin=<-1 -20 -20>, angularTarget=<0 0 0>}
      localJointOffset=<5 0 0>
     }]
    planarLimits [{drivingBone {boneName=spine_04}, planeTransform {location=<5 13.5 10>, rotation=<rot 0 0 95>}}]
    relativeSpaceBone {boneName=spine_05}
   }
   n40 = Anim Dynamics - Bone: dyn_scarf_R_04(ComponentPose=@n39, Alpha=1, GravityOverride="0,0,0", ExternalForce="0,0,0") {
    angularDampingOverride=0.9999
    angularSpringConstant=300
    bAngularSpring=true
    bOverrideAngularDamping=true
    bUseGravityOverride=true
    simulationSpace=RootRelative
    alphaBoolBlend {blendOption=Linear}
    boundBone {boneName=dyn_scarf_R_04}
    componentAppliedLinearAccClamp=<1000 1000 1000>
    componentLinearAccScale=<0.1 0.1 0.2>
    componentLinearVelScale=<0.1 0.1 0.1>
    physicsBodyDefinitions [{
      boundBone {boneName=dyn_scarf_R_04}
      boxExtents=<20 20 20>
      constraintSetup {angularLimitsMax=<1 5 25>, angularLimitsMin=<-1 -25 -25>, angularTarget=<0 0 0>}
      localJointOffset=<5 0 0>
     }]
    planarLimits [{drivingBone {boneName=spine_04}, planeTransform {location=<5 13.5 10>, rotation=<rot 0 0 95>}}]
    relativeSpaceBone {boneName=spine_05}
   }
   n41 = Anim Dynamics - Bone: dyn_fold_01_c(ComponentPose=@n40, Alpha=1, ExternalForce="0,0,0") {
    angularDampingOverride=0.99
    angularSpringConstant=300
    bAngularSpring=true
    bOverrideAngularDamping=true
    bUsePlanarLimit=false
    bUseSphericalLimits=true
    simulationSpace=RootRelative
    alphaBoolBlend {blendOption=Linear}
    boundBone {boneName=dyn_fold_01_c}
    componentAppliedLinearAccClamp=<10000 10000 10000>
    componentLinearAccScale=<0.3 0.3 0.3>
    componentLinearVelScale=<0.3 0.3 0.3>
    physicsBodyDefinitions [{
      boundBone {boneName=dyn_fold_01_c}
      boxExtents=<40 40 40>
      constraintSetup {angularLimitsMax=<0 0 15>, angularLimitsMin=<0 -15 -15>, angularTarget=<0 0 0>}
      localJointOffset=<1.5 0 0>
     }]
    sphericalLimits [{limitRadius=12.5, drivingBone {boneName=pelvis}, sphereLocalOffset=<-5 0.5 0>}]
   }
   n42 = Anim Dynamics - Bone: dyn_fold_02_c(ComponentPose=@n41, Alpha=1, ExternalForce="0,0,0") {
    angularDampingOverride=0.999
    angularSpringConstant=100
    bAngularSpring=true
    bOverrideAngularDamping=true
    bUsePlanarLimit=false
    bUseSphericalLimits=true
    simulationSpace=RootRelative
    alphaBoolBlend {blendOption=Linear}
    boundBone {boneName=dyn_fold_02_c}
    componentAppliedLinearAccClamp=<10000 10000 10000>
    componentLinearAccScale=<0.3 0.3 0.3>
    componentLinearVelScale=<0.3 0.3 0.3>
    physicsBodyDefinitions [{
      boundBone {boneName=dyn_fold_02_c}
      boxExtents=<30 30 30>
      constraintSetup {angularLimitsMax=<0 0 15>, angularLimitsMin=<0 -15 -15>, angularTarget=<0 0 0>}
      localJointOffset=<1.5 0 0>
     }]
    sphericalLimits [
     {limitRadius=8, drivingBone {boneName=thigh_l}, sphereLocalOffset=<0 5 2>}
     {limitRadius=8, drivingBone {boneName=thigh_r}, sphereLocalOffset=<0 -5 -2>}
    ]
   }
   n43 = Anim Dynamics - Bone: dyn_fold_03_c(ComponentPose=@n42, Alpha=1, ExternalForce="0,0,0") {
    angularDampingOverride=0.999
    angularSpringConstant=50
    bAngularSpring=true
    bOverrideAngularDamping=true
    bUsePlanarLimit=false
    bUseSphericalLimits=true
    simulationSpace=RootRelative
    alphaBoolBlend {blendOption=Linear}
    boundBone {boneName=dyn_fold_03_c}
    componentAppliedLinearAccClamp=<10000 10000 10000>
    componentLinearAccScale=<0.3 0.3 0.3>
    componentLinearVelScale=<0.3 0.3 0.3>
    physicsBodyDefinitions [{
      boundBone {boneName=dyn_fold_03_c}
      boxExtents=<20 20 20>
      constraintSetup {angularLimitsMax=<0 0 15>, angularLimitsMin=<0 -15 -15>, angularTarget=<0 0 0>}
      localJointOffset=<1.5 0 0>
     }]
    sphericalLimits [{limitRadius=7.5, drivingBone {boneName=thigh_l}, sphereLocalOffset=<-5 2 8>}]
   }
   n44 = Anim Dynamics - Bone: dyn_fold_04_c(ComponentPose=@n43, Alpha=1, ExternalForce="0,0,0") {
    angularDampingOverride=0.999
    angularSpringConstant=50
    bAngularSpring=true
    bOverrideAngularDamping=true
    bUsePlanarLimit=false
    bUseSphericalLimits=true
    simulationSpace=RootRelative
    alphaBoolBlend {blendOption=Linear}
    boundBone {boneName=dyn_fold_04_c}
    componentAppliedLinearAccClamp=<10000 10000 10000>
    componentLinearAccScale=<0.3 0.3 0.3>
    componentLinearVelScale=<0.3 0.3 0.3>
    physicsBodyDefinitions [{
      boundBone {boneName=dyn_fold_04_c}
      boxExtents=<20 20 20>
      constraintSetup {angularLimitsMax=<0 0 15>, angularLimitsMin=<0 -15 -15>, angularTarget=<0 0 0>}
      localJointOffset=<1.5 0 0>
     }]
    sphericalLimits [{limitRadius=8, drivingBone {boneName=thigh_r}, sphereLocalOffset=<7 -5 -2>}]
   }
   n45 = Anim Dynamics - Bone: dyn_fold_01_l(ComponentPose=@n44, Alpha=1, ExternalForce="0,0,0") {
    angularDampingOverride=0.99
    angularSpringConstant=35000
    bAngularSpring=true
    bOverrideAngularDamping=true
    simulationSpace=RootRelative
    alphaBoolBlend {blendOption=Linear}
    boundBone {boneName=dyn_fold_01_l}
    componentAppliedLinearAccClamp=<10000 10000 10000>
    componentLinearAccScale=<0.3 0.3 0.3>
    componentLinearVelScale=<0.3 0.3 0.3>
    physicsBodyDefinitions [{
      boundBone {boneName=dyn_fold_01_l}
      boxExtents=<30 30 30>
      constraintSetup {angularLimitsMax=<0 0 15>, angularLimitsMin=<0 -15 -15>, angularTarget=<0 0 0>}
      localJointOffset=<1.5 0.4 0>
     }]
    planarLimits [{drivingBone {boneName=dyn_fold_01_c}, planeTransform {location=<0 0 0.3>, rotation=<rot 0 0 0>}}]
   }
   n46 = Anim Dynamics - Bone: dyn_fold_01_r(ComponentPose=@n45, Alpha=1, ExternalForce="0,0,0") {
    angularDampingOverride=0.99
    angularSpringConstant=350
    bAngularSpring=true
    bOverrideAngularDamping=true
    simulationSpace=RootRelative
    alphaBoolBlend {blendOption=Linear}
    boundBone {boneName=dyn_fold_01_r}
    componentAppliedLinearAccClamp=<10000 10000 10000>
    componentLinearAccScale=<0.3 0.3 0.3>
    componentLinearVelScale=<0.3 0.3 0.3>
    physicsBodyDefinitions [{
      boundBone {boneName=dyn_fold_01_r}
      boxExtents=<30 30 30>
      constraintSetup {angularLimitsMax=<0 0 15>, angularLimitsMin=<0 -15 -15>, angularTarget=<0 0 0>}
      localJointOffset=<1.5 -0.4 0>
     }]
    planarLimits [{drivingBone {boneName=dyn_fold_01_c}, planeTransform {location=<0 0 0.3>, rotation=<rot 0 0 0>}}]
   }
   Save cached pose 'CtrlRigCache'(Pose=Component To Local(ComponentPose=@n46)) {cacheName=CtrlRigCache}
  ]}
 EventGraph {nodes [
   n1 = Event BlueprintUpdateAnimation()
   > n2 = Sequence!(execute) -> (then_0, then_1, then_2:Exec, then_3:Exec)
   > n3 = Gravity Override State Function DEPRECATED(CharacterPartAnimInstance=Self-Reference(), GameTimeInSeconds=GetGameTimeInSeconds(), GravityOverrideParams=Get Gravity Override Params())
   > Set Gravity Override(Gravity Override=@n3)
   Is Valid(exec=@n2.then_1, InputObject=Get PlayerPawn() -> (PlayerPawn:Fort Player Pawn Athena Object Reference))
   > Set WPO Globals Function(
    Material=Get Mat Instance Dynamic()
    Overlay Material=Get Overlay Mat Instance Dynamic()
    Skeletal Mesh=Get SkelMesh()
    "Local Noise Transform  Location"=Get Local Noise Transform Location() -> (Local Noise Transform Location:Vector)
    Local Noise Offset=Get Local Noise Offset()
   )
   > Set WPO Regional Function!(
    execute
    Material:Material Instance Dynamic Object Reference=Get Mat Instance Dynamic()
    Overlay Material:Material Instance Dynamic Object Reference=Get Overlay Mat Instance Dynamic()
    Skeletal Mesh:Skeletal Mesh Component Object Reference=Get SkelMesh()
    Velocity Transform:Name=foot_l
    Velocity Interp Speed:double=Get Velocity Interp Speed()
    Collision Transform:Name=calf_l
    Collision Transform Axis:EAxis Enum
    Invert Collision Transform Axis:bool=false
    Delta Time X:double=@n1.DeltaTimeX
    Previous Velocity Transform Location:"Vector (by ref)"=Get R1 Velocity Transform Location()
    Previous Velocity Transform Velocity:"Vector (by ref)"=Get R1 Velocity Transform Velocity()
    Region:int=1
   ) -> (then)
   > Set WPO Regional Function(
    Material=Get Mat Instance Dynamic()
    Overlay Material=Get Overlay Mat Instance Dynamic()
    Skeletal Mesh=Get SkelMesh()
    Velocity Transform=foot_r
    Velocity Interp Speed=Get Velocity Interp Speed()
    Collision Transform=calf_r
    Invert Collision Transform Axis=true
    Delta Time X=@n1.DeltaTimeX
    Previous Velocity Transform Location=Get R2 Velocity Transform Location()
    Previous Velocity Transform Velocity=Get R2 Velocity Transform Velocity()
    Region=2
   )
   > n4 = Branch(Condition=Get bIsSkydiveFloating())
   > n5 = Set WPO Pho Wind Intensity(WPO Pho Wind Intensity=1)
   SetScalarParameterValue(execute=[@n6.then, @n5.then], self=Get Mat Instance Dynamic(), ParameterName=ClothSim_WindStrength, Value=Get WPO Pho Wind Intensity())
   n6 = Set WPO Pho Wind Intensity(execute=@n4.else, WPO Pho Wind Intensity=0)
   Set CrouchAlpha(execute=@n2.then_2, CrouchAlpha:double=SelectFloat(A=1, B=0, bPickA=Get bIsCrouching())) -> (Output_Get:double)
   n7 = Branch(execute=@n2.then_3, Condition=Get isWearingBackpack() -> (isWearingBackpack:bool))
   n8 = FindChildCommponentFromAttachSocket!(
    self:Custom Character Part Anim Instance Object Reference
    Parent:Skeletal Mesh Component Object Reference=GetPartSkeletalMeshComponent()
    bIncludeAllDescendants:bool=true
    AttachSocket:Name=Backpack_BR
   ) -> (ReturnValue:Skeletal Mesh Component Object Reference)
   n9 = GetDisplayName(Object=@n8)
   n10 = Append(A=GetObjectName(Object=@n8), B=" ")
   > n11 = Branch(Condition="Equal, Case Insensitive (String)"(A=Split(SourceString=@n9, InStr=@n10, SearchDir=FromStart).RightS, B=M_MED_LazarusLens_Pack))
   > Set IsWearingWings(IsWearingWings:bool=true) -> (Output_Get:bool)
   Set IsWearingWings(execute=@n11.else, IsWearingWings=false)
   Set IsWearingWings(execute=@n7.else, IsWearingWings=false)
   Event OnCharacterCustomizationLoaded()
   > n12 = Set SkelMesh(SkelMesh=GetPartSkeletalMeshComponent())
   > n13 = Get Dynamic Material Instance(SkeletalMeshComponent=@n12)
   > Set Mat Instance Dynamic(Mat Instance Dynamic=@n13)
   > n14 = GetOrCreateDynamicOverlayMaterial(MeshComponent=@n12, MaterialName=None)
   > Set Overlay Mat Instance Dynamic(Overlay Mat Instance Dynamic=@n14)
   n15 = CheckForBackbling()
   > Set isWearingBackpack(isWearingBackpack:bool=OR Boolean!(A=@n15, B=@n15.bIsBackblingACape, C:bool=@n15.bIsBackblingLong) -> (ReturnValue)) -> (Output_Get:bool)
   Event BlueprintInitializeAnimation()
   > Set SkelMesh(SkelMesh=GetOwnerSkeletalMeshComponent())
   n16 = TryGetPawnOwner()
   > Is Valid(InputObject=@n16)
   > n17 = Cast To FortPlayerPawnAthena(Object=@n16)
   > Set PlayerPawn(PlayerPawn:Fort Player Pawn Athena Object Reference=@n17) -> (Output_Get:Fort Player Pawn Athena Object Reference)
   n18 = Branch(Condition=CharacterItemDefinitionHasAnyMetaTag(Tags=Make Array("[0]"=(TagName="Cosmetics.MetaTags.LazarusLens.Wing"))))
   > n19 = Set IsWearingWings(IsWearingWings=true)
   > PrintString!(execute, InString="To String (Boolean)"(InBool=@n19), bPrintToScreen, bPrintToLog, TextColor, Duration, Key) -> (then)
   n20 = Set IsWearingWings(execute=@n18.else, IsWearingWings=false)
   > PrintString(InString="To String (Boolean)"(InBool=@n20))
   Is Valid(InputObject=Get PlayerPawn())
   > n21 = Branch(Condition=Get bIsInFrontEnd())
   Set WPO Globals Function(
    execute=@n21.else
    Material=Get Mat Instance Dynamic()
    Overlay Material=Get Overlay Mat Instance Dynamic()
    "Local Noise Transform  Location"="0, 0, 0"
    Local Noise Offset=Get Local Noise Offset()
   )
   n22 = Reroute Node(InputPin:double) -> (OutputPin:double)
   > Set WPO Regional Function(
    Material=Get Mat Instance Dynamic()
    Overlay Material=Get Overlay Mat Instance Dynamic()
    Velocity Transform=foot_r
    Velocity Interp Speed=Get Velocity Interp Speed()
    Collision Transform=calf_r
    Invert Collision Transform Axis=false
    Delta Time X=@n22
    Previous Velocity Transform Location=Get R1 Velocity Transform Location()
    Previous Velocity Transform Velocity=Get R1 Velocity Transform Velocity()
    Region=1
   )
   > Set WPO Regional Function(
    Material=Get Mat Instance Dynamic()
    Overlay Material=Get Overlay Mat Instance Dynamic()
    Velocity Transform=foot_l
    Velocity Interp Speed=Get Velocity Interp Speed()
    Collision Transform=calf_l
    Invert Collision Transform Axis=true
    Delta Time X=@n22
    Previous Velocity Transform Location=Get R2 Velocity Transform Location()
    Previous Velocity Transform Velocity=Get R2 Velocity Transform Velocity()
    Region=2
   )
   > n23 = Branch(Condition=Get bIsSkydiveFloating())
   > n24 = Set WPO Pho Wind Intensity(WPO Pho Wind Intensity=1)
   SetScalarParameterValue(execute=[@n24.then, @n25.then], self=Get Mat Instance Dynamic(), ParameterName=ClothSim_WindStrength, Value=Get WPO Pho Wind Intensity())
   n25 = Set WPO Pho Wind Intensity(execute=@n23.else, WPO Pho Wind Intensity=0)
   Branch(Condition=Get bIsInFrontEnd())
   n26 = BreakTransform(InTransform=GetDeltaTransformFromRefPose(self=GetOwnerSkeletalMeshComponent(), BoneName=neck_01, BaseName=spine_05))
   Set NewVar(NewVar:double=MapRangeClamped(Value=BreakRotator(InRot=@n26.Rotation), InRangeA=-10, InRangeB=-40)) -> (Output_Get:double)
  ]}
]
variables [
 CrouchAlpha {type=number}
 Gravity Override {type=Vector}
 Gravity Override Params {type=GravityOverrideParamsStruct}
 isWearingBackpack {type=boolean}
 IsWearingWings {type=boolean}
 Local Noise Offset {type=Vector}
 Local Noise Transform Location {type=Vector}
 Mat Instance Dynamic {type=MaterialInstanceDynamic}
 NewVar {type=number}
 Overlay Mat Instance Dynamic {type=MaterialInstanceDynamic}
 PlayerPawn {type=FortPlayerPawnAthena}
 R1 Velocity Transform Location {type=Vector}
 R1 Velocity Transform Velocity {type=Vector}
 R2 Velocity Transform Location {type=Vector}
 R2 Velocity Transform Velocity {type=Vector}
 SkelMesh {type=SkeletalMeshComponent}
 Velocity Interp Speed {type=number, default=30}
 WPO Pho Wind Intensity {type=number}
]
