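# Physics asset of cape colliders for the LazarusLens medium male player body, apparently for cape simulation to collide against.
# Capsules cover pelvis, spine, neck, clavicles and the lower dyn_jacket bones, with a sphere on the head; arm and leg bodies have no shapes.
# Constraints follow the body hierarchy and tie the jacket chains to spine_03, but all swing and twist motions are Free, so angular limits are inactive.
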PhysicsAsset /BRCosmetics/Characters/Player/Male/Medium/Bodies/M_MED_LazarusLens/Meshes/M_MED_LazarusLens_CapeColliders
bodies {
 calf_l []
 calf_r []
 clavicle_l [{shapeType=Capsule, length=9.0192, radius=9.8871, center=<9.4266 -1.1806 -2.4977>, rotation=<rot 69.3297 1.4476 176.9462>}]
 clavicle_r [{shapeType=Capsule, length=9.0192, radius=9.8871, center=<-9.4266 1.1806 2.4977>, rotation=<rot 69.3297 1.4476 -3.0538>}]
 dyn_jacket_01_L_03 [{shapeType=Capsule, length=6.8959, radius=8.0366, shapeName=dyn_jacket_01_L_03_capsule, center=<-3.5977 0 0>, rotation=<rot 0 0 90>}]
 dyn_jacket_01_L_04 [{shapeType=Capsule, length=7.9896, radius=6.4307, shapeName=dyn_jacket_01_L_04_capsule, rotation=<rot 0 0 90>}]
 dyn_jacket_01_R_03 [{shapeType=Capsule, length=7.5907, radius=7.075, shapeName=dyn_jacket_01_R_03_capsule, center=<-3.5977 0 0>, rotation=<rot 0 0 90>}]
 dyn_jacket_01_R_04 [{shapeType=Capsule, length=8.3334, radius=6.626, shapeName=dyn_jacket_01_R_04_capsule, rotation=<rot 0 0 90>}]
 dyn_jacket_02_L_03 [{shapeType=Capsule, length=4.8893, radius=6.0706, shapeName=dyn_jacket_02_L_03_capsule, center=<-3.5977 0 0>, rotation=<rot 0 0 90>}]
 dyn_jacket_02_L_04 [{shapeType=Capsule, length=7.9769, radius=6.8581, shapeName=dyn_jacket_02_L_04_capsule, rotation=<rot 0 0 90>}]
 dyn_jacket_02_R_03 [{shapeType=Capsule, length=5.3328, radius=5.3796, shapeName=dyn_jacket_02_R_03_capsule, center=<-3.5977 0 0>, rotation=<rot 0 0 90>}]
 dyn_jacket_02_R_04 [{shapeType=Capsule, length=8.0297, radius=6.8193, shapeName=dyn_jacket_02_R_04_capsule, rotation=<rot 0 0 90>}]
 dyn_jacket_03_L_03 [{shapeType=Capsule, length=6.1909, radius=6.2388, shapeName=dyn_jacket_03_L_03_capsule, center=<-3.5977 0 0>, rotation=<rot 0 0 90>}]
 dyn_jacket_03_L_04 [{shapeType=Capsule, length=7.7595, radius=5.5498, shapeName=dyn_jacket_03_L_04_capsule, rotation=<rot -0.0001 -19.9999 90>}]
 dyn_jacket_03_R_03 [{shapeType=Capsule, length=6.2482, radius=6.2213, shapeName=dyn_jacket_03_R_03_capsule, center=<-3.5977 0 0>, rotation=<rot 0 0 90>}]
 dyn_jacket_03_R_04 [{shapeType=Capsule, length=7.6964, radius=6.4507, shapeName=dyn_jacket_03_R_04_capsule, center=<0.6616 0.2408 0>, rotation=<rot 0 20 90>}]
 dyn_jacket_B_02 [{shapeType=Capsule, length=12.857, radius=5.5161, shapeName=dyn_jacket_B_02_capsule, center=<0.5907 0 0>, rotation=<rot 0 0 90>}]
 dyn_jacket_B_L [{shapeType=Capsule, length=7.7744, radius=6.846, shapeName=dyn_jacket_B_L_capsule, center=<6.087 0 0>, rotation=<rot 0 0 90>}]
 dyn_jacket_B_R [{shapeType=Capsule, length=7.8669, radius=6.859, shapeName=dyn_jacket_B_R_capsule, center=<6.087 0 0>, rotation=<rot 0 0 90>}]
 foot_l []
 foot_r []
 head [{shapeType=Sphere, radius=17.3125, center=<11.028 -0.174 0>}]
 lowerarm_l []
 lowerarm_r []
 neck_02 [{shapeType=Capsule, length=13.4375, radius=7.3175, center=<-0.2642 -0.3408 0>, rotation=<rot 90 -171.1193 -175.3978>}]
 pelvis [{shapeType=Capsule, length=4.1456, radius=13.5388, center=<-3.197 0.7838 0>}]
 spine_01 [{shapeType=Capsule, length=5.4438, radius=13.0586, center=<-1.7965 1.1778 0>}]
 spine_02 [
  {shapeType=Capsule, length=7.3264, radius=13.6322, center=<-1.711 1.6789 0.0001>}
  {shapeType=Capsule, length=29.3956, radius=11.5807, center=<2.5024 0.4992 3.9825>, rotation=<rot 90 0.0049 0.0049>}
  {shapeType=Capsule, length=29.875, radius=11.1013, center=<2.4704 0.4991 -3.9825>, rotation=<rot 90 -0.0396 -0.0396>}
 ]
 spine_03 [{shapeType=Capsule, length=7.7107, radius=12.2031, center=<0.1445 0.5544 0.0001>}]
 spine_04 [{shapeType=Capsule, length=8.4404, radius=14.6325, center=<2.766 -1.5254 0.0002>}]
 spine_05 [
  {shapeType=Capsule, length=5.5226, radius=18.195, center=<-3.8272 1.5654 0>}
  {shapeType=Capsule, length=13.9518, radius=13.9425, center=<-0.2916 0.4452 0>}
 ]
 thigh_l []
 thigh_r []
 upperarm_l []
 upperarm_r []
}
constraints [
 {
  bone1Name=spine_01
  bone2Name=pelvis
  swing1LimitDegrees=45
  swing1Motion=Free
  swing2LimitDegrees=45
  swing2Motion=Free
  twistLimitDegrees=45
  twistMotion=Free
 }
 {
  bone1Name=spine_02
  bone2Name=spine_01
  swing1LimitDegrees=45
  swing1Motion=Free
  swing2LimitDegrees=45
  swing2Motion=Free
  twistLimitDegrees=45
  twistMotion=Free
 }
 {
  bone1Name=spine_03
  bone2Name=spine_02
  swing1LimitDegrees=45
  swing1Motion=Free
  swing2LimitDegrees=45
  swing2Motion=Free
  twistLimitDegrees=45
  twistMotion=Free
 }
 {
  bone1Name=spine_04
  bone2Name=spine_03
  swing1LimitDegrees=45
  swing1Motion=Free
  swing2LimitDegrees=45
  swing2Motion=Free
  twistLimitDegrees=45
  twistMotion=Free
 }
 {
  bone1Name=spine_05
  bone2Name=spine_04
  swing1LimitDegrees=45
  swing1Motion=Free
  swing2LimitDegrees=45
  swing2Motion=Free
  twistLimitDegrees=45
  twistMotion=Free
 }
 {
  bone1Name=neck_02
  bone2Name=spine_05
  swing1LimitDegrees=45
  swing1Motion=Free
  swing2LimitDegrees=45
  swing2Motion=Free
  twistLimitDegrees=45
  twistMotion=Free
 }
 {
  bone1Name=head
  bone2Name=neck_02
  swing1LimitDegrees=45
  swing1Motion=Free
  swing2LimitDegrees=45
  swing2Motion=Free
  twistLimitDegrees=45
  twistMotion=Free
 }
 {
  bone1Name=thigh_l
  bone2Name=pelvis
  swing1LimitDegrees=45
  swing1Motion=Free
  swing2LimitDegrees=45
  swing2Motion=Free
  twistLimitDegrees=45
  twistMotion=Free
 }
 {
  bone1Name=calf_l
  bone2Name=thigh_l
  swing1LimitDegrees=45
  swing1Motion=Free
  swing2LimitDegrees=45
  swing2Motion=Free
  twistLimitDegrees=45
  twistMotion=Free
 }
 {
  bone1Name=foot_l
  bone2Name=calf_l
  swing1LimitDegrees=45
  swing1Motion=Free
  swing2LimitDegrees=45
  swing2Motion=Free
  twistLimitDegrees=45
  twistMotion=Free
 }
 {
  bone1Name=clavicle_r
  bone2Name=spine_05
  swing1LimitDegrees=45
  swing1Motion=Free
  swing2LimitDegrees=45
  swing2Motion=Free
  twistLimitDegrees=45
  twistMotion=Free
 }
 {
  bone1Name=upperarm_r
  bone2Name=clavicle_r
  swing1LimitDegrees=45
  swing1Motion=Free
  swing2LimitDegrees=45
  swing2Motion=Free
  twistLimitDegrees=45
  twistMotion=Free
 }
 {
  bone1Name=lowerarm_r
  bone2Name=upperarm_r
  swing1LimitDegrees=45
  swing1Motion=Free
  swing2LimitDegrees=45
  swing2Motion=Free
  twistLimitDegrees=45
  twistMotion=Free
 }
 {
  bone1Name=foot_r
  bone2Name=calf_r
  swing1LimitDegrees=45
  swing1Motion=Free
  swing2LimitDegrees=45
  swing2Motion=Free
  twistLimitDegrees=45
  twistMotion=Free
 }
 {
  bone1Name=lowerarm_l
  bone2Name=upperarm_l
  swing1LimitDegrees=45
  swing1Motion=Free
  swing2LimitDegrees=45
  swing2Motion=Free
  twistLimitDegrees=45
  twistMotion=Free
 }
 {
  bone1Name=upperarm_l
  bone2Name=clavicle_l
  swing1LimitDegrees=45
  swing1Motion=Free
  swing2LimitDegrees=45
  swing2Motion=Free
  twistLimitDegrees=45
  twistMotion=Free
 }
 {
  bone1Name=clavicle_l
  bone2Name=spine_05
  swing1LimitDegrees=45
  swing1Motion=Free
  swing2LimitDegrees=45
  swing2Motion=Free
  twistLimitDegrees=45
  twistMotion=Free
 }
 {
  bone1Name=calf_r
  bone2Name=thigh_r
  swing1LimitDegrees=45
  swing1Motion=Free
  swing2LimitDegrees=45
  swing2Motion=Free
  twistLimitDegrees=45
  twistMotion=Free
 }
 {
  bone1Name=thigh_r
  bone2Name=pelvis
  swing1LimitDegrees=45
  swing1Motion=Free
  swing2LimitDegrees=45
  swing2Motion=Free
  twistLimitDegrees=45
  twistMotion=Free
 }
 {
  bone1Name=dyn_jacket_02_L_04
  bone2Name=dyn_jacket_02_L_03
  swing1LimitDegrees=45
  swing1Motion=Free
  swing2LimitDegrees=45
  swing2Motion=Free
  twistLimitDegrees=45
  twistMotion=Free
 }
 {
  bone1Name=dyn_jacket_01_R_04
  bone2Name=dyn_jacket_01_R_03
  swing1LimitDegrees=45
  swing1Motion=Free
  swing2LimitDegrees=45
  swing2Motion=Free
  twistLimitDegrees=45
  twistMotion=Free
 }
 {
  bone1Name=dyn_jacket_01_L_04
  bone2Name=dyn_jacket_01_L_03
  swing1LimitDegrees=45
  swing1Motion=Free
  swing2LimitDegrees=45
  swing2Motion=Free
  twistLimitDegrees=45
  twistMotion=Free
 }
 {
  bone1Name=dyn_jacket_02_R_04
  bone2Name=dyn_jacket_02_R_03
  swing1LimitDegrees=45
  swing1Motion=Free
  swing2LimitDegrees=45
  swing2Motion=Free
  twistLimitDegrees=45
  twistMotion=Free
 }
 {
  bone1Name=dyn_jacket_03_L_04
  bone2Name=dyn_jacket_03_L_03
  swing1LimitDegrees=45
  swing1Motion=Free
  swing2LimitDegrees=45
  swing2Motion=Free
  twistLimitDegrees=45
  twistMotion=Free
 }
 {
  bone1Name=dyn_jacket_03_R_04
  bone2Name=dyn_jacket_03_R_03
  swing1LimitDegrees=45
  swing1Motion=Free
  swing2LimitDegrees=45
  swing2Motion=Free
  twistLimitDegrees=45
  twistMotion=Free
 }
 {
  bone1Name=dyn_jacket_B_L
  bone2Name=dyn_jacket_B_02
  swing1LimitDegrees=45
  swing1Motion=Free
  swing2LimitDegrees=45
  swing2Motion=Free
  twistLimitDegrees=45
  twistMotion=Free
 }
 {
  bone1Name=dyn_jacket_B_R
  bone2Name=dyn_jacket_B_02
  swing1LimitDegrees=45
  swing1Motion=Free
  swing2LimitDegrees=45
  swing2Motion=Free
  twistLimitDegrees=45
  twistMotion=Free
 }
 {
  bone1Name=dyn_jacket_B_02
  bone2Name=spine_03
  swing1LimitDegrees=45
  swing1Motion=Free
  swing2LimitDegrees=45
  swing2Motion=Free
  twistLimitDegrees=45
  twistMotion=Free
 }
 {
  bone1Name=dyn_jacket_03_R_03
  bone2Name=spine_03
  swing1LimitDegrees=45
  swing1Motion=Free
  swing2LimitDegrees=45
  swing2Motion=Free
  twistLimitDegrees=45
  twistMotion=Free
 }
 {
  bone1Name=dyn_jacket_03_L_03
  bone2Name=spine_03
  swing1LimitDegrees=45
  swing1Motion=Free
  swing2LimitDegrees=45
  swing2Motion=Free
  twistLimitDegrees=45
  twistMotion=Free
 }
 {
  bone1Name=dyn_jacket_02_R_03
  bone2Name=spine_03
  swing1LimitDegrees=45
  swing1Motion=Free
  swing2LimitDegrees=45
  swing2Motion=Free
  twistLimitDegrees=45
  twistMotion=Free
 }
 {
  bone1Name=dyn_jacket_02_L_03
  bone2Name=spine_03
  swing1LimitDegrees=45
  swing1Motion=Free
  swing2LimitDegrees=45
  swing2Motion=Free
  twistLimitDegrees=45
  twistMotion=Free
 }
 {
  bone1Name=dyn_jacket_01_R_03
  bone2Name=spine_03
  swing1LimitDegrees=45
  swing1Motion=Free
  swing2LimitDegrees=45
  swing2Motion=Free
  twistLimitDegrees=45
  twistMotion=Free
 }
 {
  bone1Name=dyn_jacket_01_L_03
  bone2Name=spine_03
  swing1LimitDegrees=45
  swing1Motion=Free
  swing2LimitDegrees=45
  swing2Motion=Free
  twistLimitDegrees=45
  twistMotion=Free
 }
]
