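# Control Rig for the LazarusLens male face accessory that offsets the three back hair master bones as the head moves.
# Two spherical pose readers measure head orientation against spine_05, each weighting rotation and translation offsets on those bones to act as pose-driven correctives.
# The RbanCtrl bool selects between two duplicated chains that differ only in the first reader's hair rotations, with the second reader's offsets identical in both.
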ControlRigBlueprint /BRCosmetics/Characters/Player/Male/Medium/Bodies/M_MED_LazarusLens/Meshes/Parts/M_MED_LazarusLens_FaceAcc_CtrlRig
graphs [
 RigVMFunctionLibrary {links [], nodes []}
 RigVMModel {nodes [
   RigUnit_BeginExecution = &RigUnit_BeginExecution()
   RigVMFunction_ControlFlowBranch = &RigVMFunction_ControlFlowBranch(ExecuteContext=@RigUnit_BeginExecution.ExecutePin, Condition=RigVMVariableNode() {variable=RbanCtrl}.Value)
   n1 = &RigUnit_SphericalPoseReader(
    ActiveRegionScaleFactors=(PositiveWidth=0.5,NegativeWidth=0.5,PositiveHeight=1.5,NegativeHeight=1.5)
    ActiveRegionSize=0.1
    Debug=(bDrawDebug=false,DebugScale=25,DebugSegments=20,DebugThickness=0.25)
    DriverAxis=(X=0,Y=-1,Z=0)
    DriverItem=(Type=Bone,Name="head")
    FalloffRegionScaleFactors=(PositiveWidth=1.5,NegativeWidth=1.5,PositiveHeight=2,NegativeHeight=2)
    FalloffSize=0.3
    OptionalParentItem=(Type=Bone,Name="spine_05")
    RotationOffset=(X=-90,Y=0,Z=-90)
    ExecutePin=@RigVMFunction_ControlFlowBranch.False
   )
   > &RigUnit_ModifyTransforms(
    ItemToModify=((Item=(Type=Bone,Name="dyn_hair_b_mstr"),Transform=(Rotation=(X=0,Y=0,Z=0,W=1),Translation=(X=0,Y=-2,Z=2),Scale3D=(X=1,Y=1,Z=1))))
    ItemToModify.0.Transform.Rotation=&RigVMFunction_MathQuaternionFromEuler(Euler=(X=-65,Y=0,Z=0)).Result
    Weight=@n1.OutputParam
   )
   > &RigUnit_ModifyTransforms(
    ItemToModify=((Item=(Type=Bone,Name="dyn_hair_bl_mstr"),Transform=(Rotation=(X=0,Y=0,Z=0,W=1),Translation=(X=0,Y=-2,Z=2),Scale3D=(X=1,Y=1,Z=1))))
    ItemToModify.0.Transform.Rotation=&RigVMFunction_MathQuaternionFromEuler(Euler=(X=-65,Y=-30,Z=0)).Result
    Weight=@n1.OutputParam
   )
   > &RigUnit_ModifyTransforms(
    ItemToModify=((Item=(Type=Bone,Name="dyn_hair_br_mstr"),Transform=(Rotation=(X=0,Y=0,Z=0,W=1),Translation=(X=0,Y=-2,Z=2),Scale3D=(X=1,Y=1,Z=1))))
    ItemToModify.0.Transform.Rotation=&RigVMFunction_MathQuaternionFromEuler(Euler=(X=-65,Y=-30,Z=-20)).Result
    Weight=@n1.OutputParam
   )
   > n2 = &RigUnit_SphericalPoseReader(
    ActiveRegionScaleFactors=(PositiveWidth=0.3,NegativeWidth=0.3,PositiveHeight=0.3,NegativeHeight=0.3)
    ActiveRegionSize=0.1
    Debug=(bDrawDebug=false,DebugScale=25,DebugSegments=20,DebugThickness=0.25)
    DriverAxis=(X=1,Y=0,Z=0)
    DriverItem=(Type=Bone,Name="head")
    FalloffRegionScaleFactors=(PositiveWidth=1,NegativeWidth=1,PositiveHeight=1,NegativeHeight=1)
    FalloffSize=0.3
    OptionalParentItem=(Type=Bone,Name="spine_05")
    RotationOffset=(X=-20,Y=-30,Z=0)
   )
   > &RigUnit_ModifyTransforms(
    ItemToModify=((Item=(Type=Bone,Name="dyn_hair_bl_mstr"),Transform=(Rotation=(X=0,Y=0,Z=0,W=1),Translation=(X=0,Y=0,Z=-5),Scale3D=(X=1,Y=1,Z=1))))
    ItemToModify.0.Transform.Rotation=&RigVMFunction_MathQuaternionFromEuler(Euler=(X=-45,Y=0,Z=45)).Result
    Weight=@n2.OutputParam
   )
   > &RigUnit_ModifyTransforms(
    ItemToModify=((Item=(Type=Bone,Name="dyn_hair_b_mstr"),Transform=(Rotation=(X=0,Y=0,Z=0,W=1),Translation=(X=2,Y=0,Z=-2),Scale3D=(X=1,Y=1,Z=1))))
    ItemToModify.0.Transform.Rotation=&RigVMFunction_MathQuaternionFromEuler(Euler=(X=0,Y=-25,Z=0)).Result
    Weight=@n2.OutputParam
   )
   > &RigUnit_ModifyTransforms(
    ItemToModify=((Item=(Type=Bone,Name="dyn_hair_br_mstr"),Transform=(Rotation=(X=0,Y=0,Z=0,W=1),Translation=(X=2,Y=0,Z=-2),Scale3D=(X=1,Y=1,Z=1))))
    ItemToModify.0.Transform.Rotation=&RigVMFunction_MathQuaternionFromEuler(Euler=(X=0,Y=-15,Z=0)).Result
    Weight=@n2.OutputParam
   )
   n3 = &RigUnit_SphericalPoseReader(
    ActiveRegionScaleFactors=(PositiveWidth=0.5,NegativeWidth=0.5,PositiveHeight=1.5,NegativeHeight=1.5)
    ActiveRegionSize=0.1
    Debug=(bDrawDebug=false,DebugScale=25,DebugSegments=20,DebugThickness=0.25)
    DriverAxis=(X=0,Y=-1,Z=0)
    DriverItem=(Type=Bone,Name="head")
    FalloffRegionScaleFactors=(PositiveWidth=1.5,NegativeWidth=1.5,PositiveHeight=2,NegativeHeight=2)
    FalloffSize=0.3
    OptionalParentItem=(Type=Bone,Name="spine_05")
    RotationOffset=(X=-90,Y=0,Z=-90)
    ExecutePin=@RigVMFunction_ControlFlowBranch.True
   )
   RigVMFunction_MathQuaternionFromEuler = &RigVMFunction_MathQuaternionFromEuler(Euler=(X=-45,Y=-55,Z=0))
   > &RigUnit_ModifyTransforms(
    ItemToModify=((Item=(Type=Bone,Name="dyn_hair_b_mstr"),Transform=(Rotation=(X=0,Y=0,Z=0,W=1),Translation=(X=0,Y=-2,Z=2),Scale3D=(X=1,Y=1,Z=1))))
    ItemToModify.0.Transform.Rotation=@RigVMFunction_MathQuaternionFromEuler.Result
    Weight=@n3.OutputParam
   )
   > &RigUnit_ModifyTransforms(
    ItemToModify=((Item=(Type=Bone,Name="dyn_hair_bl_mstr"),Transform=(Rotation=(X=0,Y=0,Z=0,W=1),Translation=(X=0,Y=-2,Z=2),Scale3D=(X=1,Y=1,Z=1))))
    ItemToModify.0.Transform.Rotation=&RigVMFunction_MathQuaternionFromEuler(Euler=(X=-45,Y=-30,Z=0)).Result
    Weight=@n3.OutputParam
   )
   > &RigUnit_ModifyTransforms(
    ItemToModify=((Item=(Type=Bone,Name="dyn_hair_br_mstr"),Transform=(Rotation=(X=0,Y=0,Z=0,W=1),Translation=(X=0,Y=-2,Z=2),Scale3D=(X=1,Y=1,Z=1))))
    ItemToModify.0.Transform.Rotation=&RigVMFunction_MathQuaternionFromEuler(Euler=(X=-45,Y=-30,Z=-20)).Result
    Weight=@n3.OutputParam
   )
   > n4 = &RigUnit_SphericalPoseReader(
    ActiveRegionScaleFactors=(PositiveWidth=0.3,NegativeWidth=0.3,PositiveHeight=0.3,NegativeHeight=0.3)
    ActiveRegionSize=0.1
    Debug=(bDrawDebug=false,DebugScale=25,DebugSegments=20,DebugThickness=0.25)
    DriverAxis=(X=1,Y=0,Z=0)
    DriverItem=(Type=Bone,Name="head")
    FalloffRegionScaleFactors=(PositiveWidth=1,NegativeWidth=1,PositiveHeight=1,NegativeHeight=1)
    FalloffSize=0.3
    OptionalParentItem=(Type=Bone,Name="spine_05")
    RotationOffset=(X=-20,Y=-30,Z=0)
   )
   > &RigUnit_ModifyTransforms(
    ItemToModify=((Item=(Type=Bone,Name="dyn_hair_bl_mstr"),Transform=(Rotation=(X=0,Y=0,Z=0,W=1),Translation=(X=0,Y=0,Z=-5),Scale3D=(X=1,Y=1,Z=1))))
    ItemToModify.0.Transform.Rotation=&RigVMFunction_MathQuaternionFromEuler(Euler=(X=-45,Y=0,Z=45)).Result
    Weight=@n4.OutputParam
   )
   > &RigUnit_ModifyTransforms(
    ItemToModify=((Item=(Type=Bone,Name="dyn_hair_b_mstr"),Transform=(Rotation=(X=0,Y=0,Z=0,W=1),Translation=(X=2,Y=0,Z=-2),Scale3D=(X=1,Y=1,Z=1))))
    ItemToModify.0.Transform.Rotation=&RigVMFunction_MathQuaternionFromEuler(Euler=(X=0,Y=-25,Z=0)).Result
    Weight=@n4.OutputParam
   )
   > &RigUnit_ModifyTransforms(
    ItemToModify=((Item=(Type=Bone,Name="dyn_hair_br_mstr"),Transform=(Rotation=(X=0,Y=0,Z=0,W=1),Translation=(X=2,Y=0,Z=-2),Scale3D=(X=1,Y=1,Z=1))))
    ItemToModify.0.Transform.Rotation=&RigVMFunction_MathQuaternionFromEuler(Euler=(X=0,Y=-15,Z=0)).Result
    Weight=@n4.OutputParam
   )
  ]}
]
hierarchy [
 root {loc=<0 0 0>, rot=<0 0 0 1>}
 pelvis {parent=root, loc=<0 -0.7071 93.1191>, rot=<0.0127 0.707 -0.0127 -0.707>}
 spine_01 {parent=pelvis, loc=<6.7494 -0.6388 0>, rot=<0 0 0.034 -0.9994>}
 spine_02 {parent=spine_01, loc=<7.2518 -0.035 0>, rot=<0 0 0.0238 -0.9997>}
 spine_03 {parent=spine_02, loc=<6.0104 0 0>, rot=<0 0 0.0149 -0.9999>}
 spine_04 {parent=spine_03, loc=<8.1024 0 0>, rot=<0 0 -0.0132 -0.9999>}
 spine_05 {parent=spine_04, loc=<10.0146 0 0>, rot=<0 0 -0.0015 -1>}
 neck_01 {parent=spine_05, loc=<11.3138 0.1529 0>, rot=<0 0 -0.2003 -0.9797>}
 neck_02 {parent=neck_01, loc=<4.5627 0 0>, rot=<0 0 0.0504 -0.9987>}
 head {parent=neck_02, loc=<5.3914 0 0>, rot=<0 0 0.115 -0.9934>}
 dyn_hair_f_05_mstr {parent=head, loc=<14.2401 -0.01 6.3531>, rot=<0.0029 -0.7071 -0.0029 -0.7071>}
 dyn_hair_f_05_01 {parent=dyn_hair_f_05_mstr, loc=<0 0 0>, rot=<-0.4981 -0.0012 0.8588 -0.1198>}
 dyn_hair_f_05_02 {parent=dyn_hair_f_05_01, loc=<3.7822 0 0>, rot=<0 0 0 1>}
 dyn_hair_f_04_mstr {parent=head, loc=<12.1116 1.9732 6.7859>, rot=<0.0029 -0.7071 -0.0029 -0.7071>}
 dyn_hair_f_04_01 {parent=dyn_hair_f_04_mstr, loc=<0 0 0>, rot=<-0.6679 0.0495 0.7425 -0.0119>}
 dyn_hair_f_04_02 {parent=dyn_hair_f_04_01, loc=<3.8425 0 0>, rot=<0.0003 -0.0038 0.1158 -0.9933>}
 dyn_hair_f_04_03 {parent=dyn_hair_f_04_02, loc=<4.0198 0 0>, rot=<0 0 0 1>}
 dyn_hair_f_04_04 {parent=dyn_hair_f_04_02, loc=<2.5846 -2.0258 -0.1456>, rot=<0 0 0 1>}
 dyn_hair_f_03_mstr {parent=head, loc=<12.3089 4.9917 5.321>, rot=<0.0029 -0.7071 -0.0029 -0.7071>}
 dyn_hair_f_03_01 {parent=dyn_hair_f_03_mstr, loc=<0 0 0>, rot=<-0.6233 -0.074 0.7737 0.0861>}
 dyn_hair_f_03_02 {parent=dyn_hair_f_03_01, loc=<5.3993 0 0>, rot=<0.0146 -0.1127 -0.0996 -0.9885>}
 dyn_hair_f_03_03 {parent=dyn_hair_f_03_02, loc=<3.4072 0 0>, rot=<0.004 -0.0663 0.0312 -0.9973>}
 dyn_hair_f_03_04 {parent=dyn_hair_f_03_03, loc=<2.6424 0 0>, rot=<0 0 0 1>}
 dyn_hair_f_02_mstr {parent=head, loc=<12.7045 7.0633 5.7457>, rot=<0.0029 -0.7071 -0.0029 -0.7071>}
 dyn_hair_f_02_01 {parent=dyn_hair_f_02_mstr, loc=<0 0 0>, rot=<-0.158 -0.8373 -0.097 -0.5143>}
 dyn_hair_f_02_02 {parent=dyn_hair_f_02_01, loc=<4.1657 0 0>, rot=<0 0 0 1>}
 dyn_hair_f_01_mstr {parent=head, loc=<12.2837 8.0409 4.1018>, rot=<0.0029 -0.7071 -0.0029 -0.7071>}
 dyn_hair_f_01_01 {parent=dyn_hair_f_01_mstr, loc=<0 0 0>, rot=<-0.4781 0.0832 0.8325 0.2673>}
 dyn_hair_f_01_02 {parent=dyn_hair_f_01_01, loc=<4.7082 0 0>, rot=<0.0284 -0.3085 -0.1216 -0.943>}
 dyn_hair_f_01_03 {parent=dyn_hair_f_01_02, loc=<4.8174 0 0>, rot=<-0.0224 -0.1305 -0.0505 -0.9899>}
 dyn_hair_f_01_05 {parent=dyn_hair_f_01_03, loc=<2.1227 0 0>, rot=<0 0 0 1>}
 dyn_hair_f_01_04 {parent=dyn_hair_f_01_02, loc=<3.2092 -0.6217 -0.4582>, rot=<0 0 0 1>}
 dyn_hair_bl_mstr {parent=head, loc=<7.2579 -7.4422 2.864>, rot=<0.0029 -0.7071 -0.0029 -0.7071>}
 dyn_hair_bl_01 {parent=dyn_hair_bl_mstr, loc=<0 0 0>, rot=<-0.5423 -0.3267 0.6741 -0.3805>}
 dyn_hair_bl_02 {parent=dyn_hair_bl_01, loc=<3.8589 0 0>, rot=<0 0 0 1>}
 dyn_hair_bl_05 {parent=dyn_hair_bl_02, loc=<3.8308 -2.5997 -1.7681>, rot=<0.8123 -0.3161 0.0203 -0.4896>}
 dyn_hair_bl_06 {parent=dyn_hair_bl_02, loc=<5.6326 0.8634 1.8824>, rot=<0.8567 0.1421 0.1026 -0.4851>}
 dyn_hair_bl_07 {parent=dyn_hair_bl_02, loc=<6.7745 0.8914 0.3816>, rot=<0.8618 -0.1216 0.1723 -0.4613>}
 dyn_hair_bl_08 {parent=dyn_hair_bl_02, loc=<3.0881 -2.3532 0.7361>, rot=<0.8493 -0.1671 0.2231 -0.4483>}
 dyn_hair_bl_03 {parent=dyn_hair_bl_01, loc=<3.9316 -3.6615 -1.168>, rot=<0.7939 -0.3523 0.0743 -0.4899>}
 dyn_hair_bl_04 {parent=dyn_hair_bl_01, loc=<4.8387 -0.6623 2.7422>, rot=<-0.0686 0.2573 0.0431 -0.9629>}
 dyn_hair_b_mstr {parent=head, loc=<7.2636 -8.095 0>, rot=<0.0029 -0.7071 -0.0029 -0.7071>}
 dyn_hair_b_01 {parent=dyn_hair_b_mstr, loc=<0 0 0>, rot=<0.0751 -0.7031 0.0751 -0.7031>}
 dyn_hair_b_02 {parent=dyn_hair_b_01, loc=<4.6589 0 0>, rot=<0 0 0 1>}
 dyn_hair_b_05 {parent=dyn_hair_b_02, loc=<7.4991 -1.5763 2.3515>, rot=<-0.0487 0.1539 0.0924 -0.9825>}
 dyn_hair_b_06 {parent=dyn_hair_b_02, loc=<6.81 -1.8238 -1.6728>, rot=<0.043 -0.1227 0.1226 -0.9839>}
 dyn_hair_b_08 {parent=dyn_hair_b_02, loc=<7.2553 -1.6218 0.0657>, rot=<-0.0015 0.0046 0.1097 -0.9939>}
 dyn_hair_b_03 {parent=dyn_hair_b_01, loc=<6.1342 -2.8909 0.5582>, rot=<-0.0216 0.0469 0.2169 -0.9748>}
 dyn_hair_b_04 {parent=dyn_hair_b_01, loc=<6.6094 -2.4625 -3.4103>, rot=<0.0908 -0.2411 0.1415 -0.9558>}
 dyn_hair_br_mstr {parent=head, loc=<7.2579 -7.4422 -2.8922>, rot=<0.0029 -0.7071 -0.0029 -0.7071>}
 dyn_hair_br_01 {parent=dyn_hair_br_mstr, loc=<0 0 0>, rot=<-0.2188 -0.6019 0.2677 -0.7199>}
 dyn_hair_br_02 {parent=dyn_hair_br_01, loc=<3.8589 0 0>, rot=<0 0 0 1>}
 dyn_hair_br_04 {parent=dyn_hair_br_02, loc=<4.4129 5.4115 -0.5579>, rot=<0.3679 0.1341 -0.4087 -0.8244>}
 dyn_hair_br_05 {parent=dyn_hair_br_02, loc=<6.3056 2.9933 0.1341>, rot=<0.339 0.0856 -0.2026 -0.9147>}
 dyn_hair_br_06 {parent=dyn_hair_br_02, loc=<4.8876 0.4844 2.8317>, rot=<0.274 0.2607 0.0311 -0.9252>}
 dyn_hair_b_07 {parent=dyn_hair_br_02, loc=<7.9703 -1.0938 0.2132>, rot=<0.3466 -0.0112 0.0685 -0.9354>}
 dyn_hair_br_03 {parent=dyn_hair_br_01, loc=<5.255 -1.9762 3.2535>, rot=<0.277 0.2055 0.2299 -0.9101>}
 dyn_hair_t_03 {parent=head, loc=<13.6105 -8.1985 -2.0325>, rot=<0.438 -0.5551 0.4334 -0.5587>}
 dyn_hair_t_01 {parent=head, loc=<14.3225 -7.8287 -0.1849>, rot=<0.6378 -0.3053 0.6353 -0.3106>}
 dyn_hair_t_02 {parent=head, loc=<14.0177 -8.149 -1.3265>, rot=<0.5342 -0.4633 0.5303 -0.4677>}
]
variables [RbanCtrl {type=bool, default=False}]
