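# Physics asset for the LazarusLens medium male FaceAcc mesh, setting up capsule bodies for dynamic hair.
# Back, back-left and back-right hair groups branch from master bones attached to the head, with swing and tight twist limits along each branch.
# Free-angular constraints join masters to head, head to spine_05, and side groups to back at the first two joints.
# Head, upper spine and unconstrained right clavicle capsules supply body shapes around the hair.
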PhysicsAsset /BRCosmetics/Characters/Player/Male/Medium/Bodies/M_MED_LazarusLens/Meshes/Parts/M_MED_LazarusLens_FaceAcc_Physics
bodies {
 clavicle_r [{
   shapeType=Capsule
   length=5.5
   radius=8
   shapeName=clavicle_r_capsule
   center=<-11.2983 1.7838 -1.2545>
   rotation=<rot 85 -179.999 -179.9994>
  }]
 dyn_hair_b_01 [{shapeType=Capsule, length=3, radius=3, shapeName=dyn_hair_b_01_capsule, center=<2 0 0.5871>, rotation=<rot 90 0 0>}]
 dyn_hair_b_02 [{shapeType=Capsule, length=3, radius=2, shapeName=dyn_hair_b_02_capsule, center=<2 -1.5 1>, rotation=<rot 90 0 0>}]
 dyn_hair_b_03 [{shapeType=Capsule, length=2, radius=2, shapeName=dyn_hair_b_03_capsule, center=<2 0 0>, rotation=<rot 90 0 0>}]
 dyn_hair_b_04 [{shapeType=Capsule, length=2, radius=2, shapeName=dyn_hair_b_04_capsule, center=<2 0 0>, rotation=<rot 90 0 0>}]
 dyn_hair_b_05 [{shapeType=Capsule, length=2, radius=2, shapeName=dyn_hair_b_05_capsule, center=<2 0 0>, rotation=<rot 90 0 0>}]
 dyn_hair_b_06 [{shapeType=Capsule, length=2, radius=2, shapeName=dyn_hair_b_06_capsule, center=<2 0 0>, rotation=<rot 90 0 0>}]
 dyn_hair_b_07 [{shapeType=Capsule, length=2, radius=2, shapeName=dyn_hair_b_07_capsule, center=<2 0 0>, rotation=<rot 90 0 0>}]
 dyn_hair_b_08 [{shapeType=Capsule, length=2, radius=2, shapeName=dyn_hair_b_08_capsule, center=<2 0 0>, rotation=<rot 90 0 0>}]
 dyn_hair_b_09 [{shapeType=Capsule, length=2, radius=2, shapeName=dyn_hair_b_08_capsule, center=<2 0 0>, rotation=<rot 90 0 0>}]
 dyn_hair_b_10 [{shapeType=Capsule, length=2, radius=2, shapeName=dyn_hair_b_08_capsule, center=<2 0 0>, rotation=<rot 90 0 0>}]
 dyn_hair_b_mstr [{shapeType=Capsule, length=0.505, radius=0.505, shapeName=dyn_hair_b_mstr_capsule, rotation=<rot 0 -90 0>}]
 dyn_hair_bl_01 [{shapeType=Capsule, length=2, radius=5, shapeName=dyn_hair_l_01_capsule, center=<2 0 1.4947>, rotation=<rot 90 0 0>}]
 dyn_hair_bl_02 [{shapeType=Capsule, length=2, radius=2, shapeName=dyn_hair_l_02_capsule, center=<2 0 0>, rotation=<rot 90 0 0>}]
 dyn_hair_bl_03 [{
   shapeType=Capsule
   length=3
   radius=3
   shapeName=dyn_hair_bl_03_capsule
   center=<1.9249 -0.7447 0.4666>
   rotation=<rot 78.3739 152.5327 179.9999>
  }]
 dyn_hair_bl_04 [{shapeType=Capsule, length=2, radius=2, shapeName=dyn_hair_bl_04_capsule, center=<2 0 0>, rotation=<rot 90 0 0>}]
 dyn_hair_bl_05 [{
   shapeType=Capsule
   length=3
   radius=2.5
   shapeName=dyn_hair_bl_05_capsule
   center=<1.9123 -0.9934 -0.5127>
   rotation=<rot 81.6274 -28.8004 0>
  }]
 dyn_hair_bl_08 [{shapeType=Capsule, length=2, radius=2, shapeName=dyn_hair_bl_08_capsule, center=<2 0 0>, rotation=<rot 90 0 0>}]
 dyn_hair_bl_mstr [{shapeType=Capsule, length=0.505, radius=0.505, shapeName=dyn_hair_bl_mstr_capsule, rotation=<rot 0 -90 0>}]
 dyn_hair_br_01 [{shapeType=Capsule, length=5, radius=3, shapeName=dyn_hair_l_01_capsule, center=<2.3433 0 2.4834>, rotation=<rot 78.7345 0 -17.3747>}]
 dyn_hair_br_02 [{shapeType=Capsule, length=4, radius=3, shapeName=dyn_hair_l_02_capsule, center=<2 0 1>, rotation=<rot 90 0 0>}]
 dyn_hair_br_03 [{shapeType=Capsule, length=2, radius=4, shapeName=dyn_hair_br_03_capsule, center=<2 0 0>, rotation=<rot 90 0 0>}]
 dyn_hair_br_04 [{shapeType=Capsule, length=2, radius=2, shapeName=dyn_hair_br_04_capsule, center=<2 0 0>, rotation=<rot 90 0 0>}]
 dyn_hair_br_05 [{shapeType=Capsule, length=2, radius=2, shapeName=dyn_hair_br_05_capsule, center=<2 0 0>, rotation=<rot 90 0 0>}]
 dyn_hair_br_06 [{shapeType=Capsule, length=2, radius=2, shapeName=dyn_hair_br_06_capsule, center=<2 0 0>, rotation=<rot 90 0 0>}]
 dyn_hair_br_mstr [{shapeType=Capsule, length=0.505, radius=0.505, shapeName=dyn_hair_br_mstr_capsule, rotation=<rot 0 -90 0>}]
 head [{shapeType=Capsule, length=12, radius=9, shapeName=head_capsule, center=<2.5 1.5 0>, rotation=<rot 90 0 0>}]
 spine_05 [{shapeType=Capsule, length=30, radius=10, shapeName=spine_05_capsule, center=<5 0 0>, rotation=<rot 0 -90 0>}]
}
constraints [
 {bone1Name=dyn_hair_br_03, bone2Name=dyn_hair_br_01, swing1LimitDegrees=25, swing2LimitDegrees=25, twistLimitDegrees=2}
 {
  bone1Name=dyn_hair_b_mstr
  bone2Name=head
  swing1LimitDegrees=45
  swing1Motion=Free
  swing2LimitDegrees=45
  swing2Motion=Free
  twistLimitDegrees=45
  twistMotion=Free
 }
 {bone1Name=dyn_hair_b_01, bone2Name=dyn_hair_b_mstr, swing1LimitDegrees=25, swing2LimitDegrees=10, twistLimitDegrees=2}
 {bone1Name=dyn_hair_b_02, bone2Name=dyn_hair_b_01, swing1LimitDegrees=30, swing2LimitDegrees=10, twistLimitDegrees=2}
 {bone1Name=dyn_hair_b_05, bone2Name=dyn_hair_b_02, swing1LimitDegrees=25, swing2LimitDegrees=25, twistLimitDegrees=2}
 {bone1Name=dyn_hair_b_06, bone2Name=dyn_hair_b_02, swing1LimitDegrees=25, swing2LimitDegrees=25, twistLimitDegrees=2}
 {bone1Name=dyn_hair_b_08, bone2Name=dyn_hair_b_02, swing1LimitDegrees=25, swing2LimitDegrees=25, twistLimitDegrees=2}
 {bone1Name=dyn_hair_b_03, bone2Name=dyn_hair_b_01, swing1LimitDegrees=25, swing2LimitDegrees=25, twistLimitDegrees=2}
 {bone1Name=dyn_hair_b_04, bone2Name=dyn_hair_b_01, swing1LimitDegrees=25, swing2LimitDegrees=15, twistLimitDegrees=2}
 {
  bone1Name=dyn_hair_bl_mstr
  bone2Name=head
  swing1LimitDegrees=45
  swing1Motion=Free
  swing2LimitDegrees=45
  swing2Motion=Free
  twistLimitDegrees=45
  twistMotion=Free
 }
 {bone1Name=dyn_hair_bl_01, bone2Name=dyn_hair_bl_mstr, swing1LimitDegrees=10, swing2LimitDegrees=25, twistLimitDegrees=5}
 {bone1Name=dyn_hair_bl_02, bone2Name=dyn_hair_bl_01, swing1LimitDegrees=10, swing2LimitDegrees=30, twistLimitDegrees=5}
 {bone1Name=dyn_hair_bl_05, bone2Name=dyn_hair_bl_02, swing1LimitDegrees=35, swing2LimitDegrees=35, twistLimitDegrees=5}
 {bone1Name=dyn_hair_bl_08, bone2Name=dyn_hair_bl_02, swing1LimitDegrees=15, swing2LimitDegrees=15, twistLimitDegrees=5}
 {bone1Name=dyn_hair_bl_03, bone2Name=dyn_hair_bl_01, swing1LimitDegrees=30, swing2LimitDegrees=30, twistLimitDegrees=5}
 {bone1Name=dyn_hair_bl_04, bone2Name=dyn_hair_bl_01, swing1LimitDegrees=20, swing2LimitDegrees=20, twistLimitDegrees=5}
 {
  bone1Name=dyn_hair_br_mstr
  bone2Name=head
  swing1LimitDegrees=45
  swing1Motion=Free
  swing2LimitDegrees=45
  swing2Motion=Free
  twistLimitDegrees=45
  twistMotion=Free
 }
 {bone1Name=dyn_hair_br_01, bone2Name=dyn_hair_br_mstr, swing1LimitDegrees=15, swing2LimitDegrees=15, twistLimitDegrees=2}
 {bone1Name=dyn_hair_br_02, bone2Name=dyn_hair_br_01, swing1LimitDegrees=20, swing2LimitDegrees=20, twistLimitDegrees=2}
 {bone1Name=dyn_hair_b_07, bone2Name=dyn_hair_br_02, swing1LimitDegrees=25, swing2LimitDegrees=15, twistLimitDegrees=2}
 {bone1Name=dyn_hair_br_04, bone2Name=dyn_hair_br_02, swing1LimitDegrees=25, swing2LimitDegrees=25, twistLimitDegrees=2}
 {bone1Name=dyn_hair_br_05, bone2Name=dyn_hair_br_02, swing1LimitDegrees=25, swing2LimitDegrees=25, twistLimitDegrees=2}
 {bone1Name=dyn_hair_br_06, bone2Name=dyn_hair_br_02, swing1LimitDegrees=25, swing2LimitDegrees=25, twistLimitDegrees=2}
 {
  bone1Name=head
  bone2Name=spine_05
  swing1LimitDegrees=45
  swing1Motion=Free
  swing2LimitDegrees=45
  swing2Motion=Free
  twistLimitDegrees=45
  twistMotion=Free
 }
 {bone1Name=dyn_hair_b_10, bone2Name=dyn_hair_b_02, swing1LimitDegrees=25, swing2LimitDegrees=25, twistLimitDegrees=2}
 {bone1Name=dyn_hair_b_09, bone2Name=dyn_hair_b_02, swing1LimitDegrees=25, swing2LimitDegrees=25, twistLimitDegrees=2}
 {
  bone1Name=dyn_hair_bl_01
  bone2Name=dyn_hair_b_01
  swing1LimitDegrees=45
  swing1Motion=Free
  swing2LimitDegrees=45
  swing2Motion=Free
  twistLimitDegrees=45
  twistMotion=Free
 }
 {
  bone1Name=dyn_hair_br_01
  bone2Name=dyn_hair_b_01
  swing1LimitDegrees=45
  swing1Motion=Free
  swing2LimitDegrees=45
  swing2Motion=Free
  twistLimitDegrees=45
  twistMotion=Free
 }
 {
  bone1Name=dyn_hair_bl_02
  bone2Name=dyn_hair_b_02
  swing1LimitDegrees=45
  swing1Motion=Free
  swing2LimitDegrees=45
  swing2Motion=Free
  twistLimitDegrees=45
  twistMotion=Free
 }
 {
  bone1Name=dyn_hair_br_02
  bone2Name=dyn_hair_b_02
  swing1LimitDegrees=45
  swing1Motion=Free
  swing2LimitDegrees=45
  swing2Motion=Free
  twistLimitDegrees=45
  twistMotion=Free
 }
]
