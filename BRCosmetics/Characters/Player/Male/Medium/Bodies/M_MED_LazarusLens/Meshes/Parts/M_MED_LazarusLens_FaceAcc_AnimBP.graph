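# Anim Blueprint for the LazarusLens face accessory, adding secondary physics to its dyn_hair bones.
# The parent mesh pose runs through a Control Rig into a cache feeding two Anim Dynamics chains, blended by bool.
# One chain adds RigidBody and an averaging constraint; the other pins f_01 bones to f_03.
# A deprecated gravity-override function updates both chains' external force each frame; a third, older chain is left unconnected.
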AnimBlueprint /BRCosmetics/Characters/Player/Male/Medium/Bodies/M_MED_LazarusLens/Meshes/Parts/M_MED_LazarusLens_FaceAcc_AnimBP
target_skeleton=&/M_MED_LazarusLens_FaceAcc_Skeleton
graphs [
 AnimGraph {nodes [
   n1 = Anim Dynamics - Bone: dyn_hair_f_01_01(Alpha=1, ExternalForce="0,0,0") {
    angularDampingOverride=0.9999
    angularSpringConstant=150
    bAngularSpring=true
    bOverrideAngularDamping=true
    simulationSpace=RootRelative
    alphaBoolBlend {blendOption=Linear}
    boundBone {boneName=dyn_hair_f_01_01}
    componentAppliedLinearAccClamp=<10000 10000 10000>
    componentLinearAccScale=<0.5 0.5 0.5>
    componentLinearVelScale=<0.5 0.5 0.5>
    physicsBodyDefinitions [{
      boundBone {boneName=dyn_hair_f_01_01}
      boxExtents=<40 40 40>
      constraintSetup {angularLimitsMax=<1 0 5>, angularLimitsMin=<-1 -15 -5>, angularTarget=<0 0 0>}
      localJointOffset=<2 0 0>
     }]
    relativeSpaceBone {boneName=head}
   }
   n2 = Anim Dynamics - Bone: dyn_hair_f_01_02(ComponentPose=@n1, Alpha=1, ExternalForce="0,0,0") {
    angularDampingOverride=0.999
    angularSpringConstant=100
    bOverrideAngularDamping=true
    bUsePlanarLimit=false
    bUseSphericalLimits=true
    simulationSpace=RootRelative
    alphaBoolBlend {blendOption=Linear}
    boundBone {boneName=dyn_hair_f_01_02}
    componentAppliedLinearAccClamp=<10000 10000 10000>
    componentLinearAccScale=<0.5 0.5 0.5>
    componentLinearVelScale=<0.5 0.5 0.5>
    physicsBodyDefinitions [{
      boundBone {boneName=dyn_hair_f_01_02}
      boxExtents=<30 30 30>
      constraintSetup {angularLimitsMax=<1 3 10>, angularLimitsMin=<-1 -15 -10>, angularTarget=<0 0 0>}
      localJointOffset=<1.5 0 0>
     }]
    relativeSpaceBone {boneName=head}
    sphericalLimits [{limitRadius=7, drivingBone {boneName=head}, sphereLocalOffset=<5 10 0>}]
   }
   n3 = Anim Dynamics - Bone: dyn_hair_f_01_03(ComponentPose=@n2, Alpha=1, ExternalForce="0,0,0") {
    angularDampingOverride=0.99
    angularSpringConstant=50
    bAngularSpring=true
    bOverrideAngularDamping=true
    simulationSpace=RootRelative
    alphaBoolBlend {blendOption=Linear}
    boundBone {boneName=dyn_hair_f_01_03}
    componentAppliedLinearAccClamp=<10000 10000 10000>
    componentLinearAccScale=<0.5 0.5 0.5>
    componentLinearVelScale=<0.5 0.5 0.5>
    physicsBodyDefinitions [{
      boundBone {boneName=dyn_hair_f_01_03}
      boxExtents=<20 20 20>
      constraintSetup {angularLimitsMax=<1 5 15>, angularLimitsMin=<-1 -15 -15>, angularTarget=<0 0 0>}
      localJointOffset=<1.5 0 0>
     }]
    relativeSpaceBone {boneName=head}
   }
   n4 = Anim Dynamics - Bone: dyn_hair_f_01_04(ComponentPose=@n3, Alpha=1, ExternalForce="0,0,0") {
    angularDampingOverride=0.9
    angularSpringConstant=25
    bAngularSpring=true
    bOverrideAngularDamping=true
    simulationSpace=RootRelative
    alphaBoolBlend {blendOption=Linear}
    boundBone {boneName=dyn_hair_f_01_04}
    componentAppliedLinearAccClamp=<10000 10000 10000>
    componentLinearAccScale=<0.5 0.5 0.5>
    componentLinearVelScale=<0.5 0.5 0.5>
    physicsBodyDefinitions [{
      boundBone {boneName=dyn_hair_f_01_04}
      boxExtents=<15 15 15>
      constraintSetup {angularLimitsMax=<1 10 20>, angularLimitsMin=<-1 -20 -20>, angularTarget=<0 0 0>}
      localJointOffset=<1.5 0 0>
     }]
    relativeSpaceBone {boneName=head}
   }
   Anim Dynamics - Bone: dyn_hair_f_01_05(ComponentPose=@n4, Alpha=1, ExternalForce="0,0,0") {
    angularDampingOverride=0.9
    angularSpringConstant=25
    bAngularSpring=true
    bOverrideAngularDamping=true
    simulationSpace=RootRelative
    alphaBoolBlend {blendOption=Linear}
    boundBone {boneName=dyn_hair_f_01_05}
    componentAppliedLinearAccClamp=<10000 10000 10000>
    componentLinearAccScale=<0.5 0.5 0.5>
    componentLinearVelScale=<0.5 0.5 0.5>
    physicsBodyDefinitions [{
      boundBone {boneName=dyn_hair_f_01_05}
      boxExtents=<15 15 15>
      constraintSetup {angularLimitsMax=<1 10 20>, angularLimitsMin=<-1 -20 -20>, angularTarget=<0 0 0>}
      localJointOffset=<1.5 0 0>
     }]
    relativeSpaceBone {boneName=head}
   }
   "Constraint - dyn_hair_f_01_05 (P)"(Alpha=1, ConstraintWeights_0=1) {
    alphaBoolBlend {blendOption=Linear}
    boneToModify {boneName=dyn_hair_f_01_05}
    constraintSetup [{targetBone {boneName=dyn_hair_f_03_04}}]
    constraintWeights [1]
   }
   "Constraint - dyn_hair_f_01_03 (P)"(Alpha=1, ConstraintWeights_0=1) {
    alphaBoolBlend {blendOption=Linear}
    boneToModify {boneName=dyn_hair_f_01_03}
    constraintSetup [{targetBone {boneName=dyn_hair_f_03_03}}]
    constraintWeights [1]
   }
   n5 = RigidBody!(ComponentPose=Local To Component(LocalPose=Use cached pose 'PostCtrlRig'()), Alpha, OverrideWorldGravity:Vector="0,0,0", ExternalForce) -> (Pose) {
    bOverrideWorldGravity=true
    overridePhysicsAsset=&/M_MED_LazarusLens_FaceAcc_Physics
    simulationSpace=BaseBoneSpace
    alphaBoolBlend {blendOption=Linear}
    baseBoneRef {boneName=spine_05}
    componentAppliedLinearAccClamp=<2500 2500 2500>
    componentLinearAccScale=<0.35 0.3 0.1>
    componentLinearVelScale=<0.25 0.15 0.1>
    simSpaceSettings {velocityScaleZ=0.1, worldAlpha=0.6}
   }
   n6 = Get GravityOverride()
   n7 = Anim Dynamics - Bone: dyn_hair_f_01_01(ComponentPose=@n5, Alpha=1, ExternalForce=@n6) {
    angularDampingOverride=0.99
    angularSpringConstant=100
    bOverrideAngularDamping=true
    bUseGravityOverride=true
    simulationSpace=RootRelative
    alphaBoolBlend {blendOption=Linear}
    boundBone {boneName=dyn_hair_f_01_01}
    componentAppliedLinearAccClamp=<10000 10000 10000>
    componentLinearAccScale=<1 1 1>
    componentLinearVelScale=<1 1 1>
    physicsBodyDefinitions [{
      boundBone {boneName=dyn_hair_f_01_01}
      boxExtents=<20 20 20>
      constraintSetup {angularLimitsMax=<1 0 10>, angularLimitsMin=<-1 -20 -10>, angularTarget=<0 0 0>}
      localJointOffset=<1 0 0>
     }]
    relativeSpaceBone {boneName=spine_05}
   }
   n8 = Anim Dynamics - Bone: dyn_hair_f_01_02(ComponentPose=@n7, Alpha=1, ExternalForce=@n6) {
    angularDampingOverride=0.99
    angularSpringConstant=75
    bOverrideAngularDamping=true
    bUseSphericalLimits=true
    simulationSpace=RootRelative
    alphaBoolBlend {blendOption=Linear}
    boundBone {boneName=dyn_hair_f_01_02}
    componentAppliedLinearAccClamp=<10000 10000 10000>
    componentLinearAccScale=<2 2 2>
    componentLinearVelScale=<2 2 2>
    physicsBodyDefinitions [{
      boundBone {boneName=dyn_hair_f_01_02}
      boxExtents=<20 20 20>
      constraintSetup {angularLimitsMax=<1 3 15>, angularLimitsMin=<-1 -20 -15>, angularTarget=<0 0 0>}
      localJointOffset=<1 0 0>
     }]
    relativeSpaceBone {boneName=spine_05}
    sphericalLimits [{limitRadius=4.6, drivingBone {boneName=head}, sphereLocalOffset=<5 8 3>}]
   }
   n9 = Anim Dynamics - Bone: dyn_hair_f_01_03(ComponentPose=@n8, Alpha=1, ExternalForce=@n6) {
    angularDampingOverride=0.99
    angularSpringConstant=50
    bOverrideAngularDamping=true
    bUsePlanarLimit=false
    bUseSphericalLimits=true
    simulationSpace=RootRelative
    alphaBoolBlend {blendOption=Linear}
    boundBone {boneName=dyn_hair_f_01_03}
    componentAppliedLinearAccClamp=<10000 10000 10000>
    componentLinearAccScale=<2 2 2>
    componentLinearVelScale=<2 2 2>
    physicsBodyDefinitions [{
      boundBone {boneName=dyn_hair_f_01_03}
      boxExtents=<15 15 15>
      constraintSetup {angularLimitsMax=<1 5 15>, angularLimitsMin=<-1 -15 -15>, angularTarget=<0 0 0>}
      localJointOffset=<1 0 0>
     }]
    relativeSpaceBone {boneName=spine_05}
    sphericalLimits [{limitRadius=4.6, drivingBone {boneName=head}, sphereLocalOffset=<5 7 3>}]
   }
   n10 = Anim Dynamics - Bone: dyn_hair_f_01_05(ComponentPose=@n9, Alpha=1, ExternalForce=@n6) {
    angularDampingOverride=0.9
    angularSpringConstant=50
    bOverrideAngularDamping=true
    bUsePlanarLimit=false
    bUseSphericalLimits=true
    simulationSpace=RootRelative
    alphaBoolBlend {blendOption=Linear}
    boundBone {boneName=dyn_hair_f_01_05}
    componentAppliedLinearAccClamp=<10000 10000 10000>
    componentLinearAccScale=<2 2 2>
    componentLinearVelScale=<2 2 2>
    physicsBodyDefinitions [{
      boundBone {boneName=dyn_hair_f_01_05}
      boxExtents=<15 15 15>
      constraintSetup {angularLimitsMax=<1 5 15>, angularLimitsMin=<-1 -15 -15>, angularTarget=<0 0 0>}
      localJointOffset=<1 0 0>
     }]
    relativeSpaceBone {boneName=spine_05}
    sphericalLimits [{limitRadius=4, drivingBone {boneName=head}, sphereLocalOffset=<2 9 3>}]
   }
   n11 = Anim Dynamics - Bone: dyn_hair_f_03_01(ComponentPose=@n10, Alpha=1, ExternalForce=@n6) {
    angularDampingOverride=0.99
    angularSpringConstant=100
    bOverrideAngularDamping=true
    simulationSpace=RootRelative
    alphaBoolBlend {blendOption=Linear}
    boundBone {boneName=dyn_hair_f_03_01}
    componentAppliedLinearAccClamp=<10000 10000 10000>
    componentLinearAccScale=<1 1 1>
    componentLinearVelScale=<1 1 1>
    physicsBodyDefinitions [{
      boundBone {boneName=dyn_hair_f_03_01}
      boxExtents=<20 20 20>
      constraintSetup {angularLimitsMax=<1 0 10>, angularLimitsMin=<-1 -20 -10>, angularTarget=<0 0 0>}
      localJointOffset=<1 0 0>
     }]
    planarLimits [{drivingBone {boneName=head}, planeTransform {location=<5 0 4>, rotation=<rot 0 0 0>}}]
    relativeSpaceBone {boneName=spine_05}
    sphericalLimits [{limitRadius=5.5, drivingBone {boneName=head}, sphereLocalOffset=<5 5 3>}]
   }
   n12 = Anim Dynamics - Bone: dyn_hair_f_03_02(ComponentPose=@n11, Alpha=1, ExternalForce=@n6) {
    angularDampingOverride=0.99
    angularSpringConstant=75
    bOverrideAngularDamping=true
    simulationSpace=RootRelative
    alphaBoolBlend {blendOption=Linear}
    boundBone {boneName=dyn_hair_f_03_02}
    componentAppliedLinearAccClamp=<10000 10000 10000>
    componentLinearAccScale=<2 2 2>
    componentLinearVelScale=<2 2 2>
    physicsBodyDefinitions [{
      boundBone {boneName=dyn_hair_f_03_02}
      boxExtents=<20 20 20>
      constraintSetup {angularLimitsMax=<1 3 15>, angularLimitsMin=<-1 -20 -15>, angularTarget=<0 0 0>}
      localJointOffset=<1 0 0>
     }]
    planarLimits [{drivingBone {boneName=head}, planeTransform {location=<5 0 7>, rotation=<rot 0 0 0>}}]
    relativeSpaceBone {boneName=spine_05}
    sphericalLimits [{limitRadius=6, drivingBone {boneName=head}, sphereLocalOffset=<5 5 3>}]
   }
   n13 = Anim Dynamics - Bone: dyn_hair_f_03_03(ComponentPose=@n12, Alpha=1, ExternalForce=@n6) {
    angularDampingOverride=0.99
    angularSpringConstant=50
    bOverrideAngularDamping=true
    bUsePlanarLimit=false
    bUseSphericalLimits=true
    simulationSpace=RootRelative
    alphaBoolBlend {blendOption=Linear}
    boundBone {boneName=dyn_hair_f_03_03}
    componentAppliedLinearAccClamp=<10000 10000 10000>
    componentLinearAccScale=<2 2 2>
    componentLinearVelScale=<2 2 2>
    physicsBodyDefinitions [{
      boundBone {boneName=dyn_hair_f_03_03}
      boxExtents=<15 15 15>
      constraintSetup {angularLimitsMax=<1 5 15>, angularLimitsMin=<-1 -15 -15>, angularTarget=<0 0 0>}
      localJointOffset=<1 0 0>
     }]
    relativeSpaceBone {boneName=spine_05}
    sphericalLimits [{limitRadius=4.6, drivingBone {boneName=head}, sphereLocalOffset=<5 7 3>}]
   }
   n14 = Anim Dynamics - Bone: dyn_hair_f_03_04(ComponentPose=@n13, Alpha=1, ExternalForce=@n6) {
    angularDampingOverride=0.9
    angularSpringConstant=50
    bOverrideAngularDamping=true
    bUsePlanarLimit=false
    bUseSphericalLimits=true
    simulationSpace=RootRelative
    alphaBoolBlend {blendOption=Linear}
    boundBone {boneName=dyn_hair_f_03_04}
    componentAppliedLinearAccClamp=<10000 10000 10000>
    componentLinearAccScale=<2 2 2>
    componentLinearVelScale=<2 2 2>
    physicsBodyDefinitions [{
      boundBone {boneName=dyn_hair_f_03_04}
      boxExtents=<15 15 15>
      constraintSetup {angularLimitsMax=<1 5 15>, angularLimitsMin=<-1 -15 -15>, angularTarget=<0 0 0>}
      localJointOffset=<1 0 0>
     }]
    relativeSpaceBone {boneName=spine_05}
    sphericalLimits [{limitRadius=4, drivingBone {boneName=head}, sphereLocalOffset=<2 5 3>}]
   }
   n15 = Anim Dynamics - Bone: dyn_hair_f_04_01(ComponentPose=@n14, Alpha=1, ExternalForce=@n6) {
    angularDampingOverride=0.99
    angularSpringConstant=100
    bOverrideAngularDamping=true
    simulationSpace=RootRelative
    alphaBoolBlend {blendOption=Linear}
    boundBone {boneName=dyn_hair_f_04_01}
    componentAppliedLinearAccClamp=<10000 10000 10000>
    componentLinearAccScale=<1 1 1>
    componentLinearVelScale=<1 1 1>
    physicsBodyDefinitions [{
      boundBone {boneName=dyn_hair_f_04_01}
      boxExtents=<20 20 20>
      constraintSetup {angularLimitsMax=<1 0 10>, angularLimitsMin=<-1 -20 -10>, angularTarget=<0 0 0>}
      localJointOffset=<1 0 0>
     }]
    planarLimits [{drivingBone {boneName=head}, planeTransform {location=<5 0 4>, rotation=<rot 0 0 0>}}]
    relativeSpaceBone {boneName=spine_05}
   }
   n16 = Anim Dynamics - Bone: dyn_hair_f_04_02(ComponentPose=@n15, Alpha=1, ExternalForce=@n6) {
    angularDampingOverride=0.99
    angularSpringConstant=75
    bOverrideAngularDamping=true
    simulationSpace=RootRelative
    alphaBoolBlend {blendOption=Linear}
    boundBone {boneName=dyn_hair_f_04_02}
    componentAppliedLinearAccClamp=<10000 10000 10000>
    componentLinearAccScale=<2 2 2>
    componentLinearVelScale=<2 2 2>
    physicsBodyDefinitions [{
      boundBone {boneName=dyn_hair_f_04_02}
      boxExtents=<20 20 20>
      constraintSetup {angularLimitsMax=<1 3 15>, angularLimitsMin=<-1 -20 -15>, angularTarget=<0 0 0>}
      localJointOffset=<1 0 0>
     }]
    planarLimits [{drivingBone {boneName=head}, planeTransform {location=<5 0 7>, rotation=<rot 0 0 0>}}]
    relativeSpaceBone {boneName=spine_05}
   }
   n17 = Anim Dynamics - Bone: dyn_hair_f_04_03(ComponentPose=@n16, Alpha=1, ExternalForce=@n6) {
    angularDampingOverride=0.99
    angularSpringConstant=50
    bOverrideAngularDamping=true
    simulationSpace=RootRelative
    alphaBoolBlend {blendOption=Linear}
    boundBone {boneName=dyn_hair_f_04_03}
    componentAppliedLinearAccClamp=<10000 10000 10000>
    componentLinearAccScale=<2 2 2>
    componentLinearVelScale=<2 2 2>
    physicsBodyDefinitions [{
      boundBone {boneName=dyn_hair_f_04_03}
      boxExtents=<15 15 15>
      constraintSetup {angularLimitsMax=<1 5 15>, angularLimitsMin=<-1 -15 -15>, angularTarget=<0 0 0>}
      localJointOffset=<1 0 0>
     }]
    relativeSpaceBone {boneName=spine_05}
   }
   n18 = Anim Dynamics - Bone: dyn_hair_f_04_04(ComponentPose=@n17, Alpha=1, ExternalForce=@n6) {
    angularDampingOverride=0.9
    angularSpringConstant=50
    bOverrideAngularDamping=true
    simulationSpace=RootRelative
    alphaBoolBlend {blendOption=Linear}
    boundBone {boneName=dyn_hair_f_04_04}
    componentAppliedLinearAccClamp=<10000 10000 10000>
    componentLinearAccScale=<2 2 2>
    componentLinearVelScale=<2 2 2>
    physicsBodyDefinitions [{
      boundBone {boneName=dyn_hair_f_04_04}
      boxExtents=<15 15 15>
      constraintSetup {angularLimitsMax=<1 5 15>, angularLimitsMin=<-1 -15 -15>, angularTarget=<0 0 0>}
      localJointOffset=<1 0 0>
     }]
    relativeSpaceBone {boneName=spine_05}
   }
   n19 = Anim Dynamics - Bone: dyn_hair_f_05_01(ComponentPose=@n18, Alpha=1, ExternalForce=@n6) {
    angularDampingOverride=0.999
    angularSpringConstant=50
    bAngularSpring=true
    bOverrideAngularDamping=true
    simulationSpace=RootRelative
    alphaBoolBlend {blendOption=Linear}
    boundBone {boneName=dyn_hair_f_05_01}
    componentAppliedLinearAccClamp=<10000 10000 10000>
    componentLinearAccScale=<0.2 0.2 0.2>
    componentLinearVelScale=<0.2 0.2 0.2>
    physicsBodyDefinitions [{
      boundBone {boneName=dyn_hair_f_05_01}
      boxExtents=<20 20 20>
      constraintSetup {angularLimitsMax=<2 3 15>, angularLimitsMin=<-2 -10 -15>, angularTarget=<0 0 0>}
      localJointOffset=<1 0 0>
     }]
    relativeSpaceBone {boneName=head}
   }
   n20 = Anim Dynamics - Bone: dyn_hair_f_05_02(ComponentPose=@n19, Alpha=1, ExternalForce=@n6) {
    angularDampingOverride=0.99
    angularSpringConstant=50
    bAngularSpring=true
    bOverrideAngularDamping=true
    simulationSpace=RootRelative
    alphaBoolBlend {blendOption=Linear}
    boundBone {boneName=dyn_hair_f_05_02}
    componentAppliedLinearAccClamp=<10000 10000 10000>
    componentLinearAccScale=<0.3 0.3 0.3>
    componentLinearVelScale=<0.3 0.3 0.3>
    physicsBodyDefinitions [{
      boundBone {boneName=dyn_hair_f_05_02}
      boxExtents=<15 15 15>
      constraintSetup {angularLimitsMax=<2 20 20>, angularLimitsMin=<-2 -20 -20>, angularTarget=<0 0 0>}
      localJointOffset=<1 0 0>
     }]
    relativeSpaceBone {boneName=head}
   }
   n21 = Anim Dynamics - Bone: dyn_hair_f_02_02(ComponentPose=@n20, Alpha=1, ExternalForce=@n6) {
    angularDampingOverride=0.99
    angularSpringConstant=50
    bAngularSpring=true
    bOverrideAngularDamping=true
    simulationSpace=RootRelative
    alphaBoolBlend {blendOption=Linear}
    boundBone {boneName=dyn_hair_f_02_02}
    componentAppliedLinearAccClamp=<10000 10000 10000>
    componentLinearAccScale=<0.3 0.3 0.3>
    componentLinearVelScale=<0.3 0.3 0.3>
    physicsBodyDefinitions [{
      boundBone {boneName=dyn_hair_f_02_02}
      boxExtents=<20 20 20>
      constraintSetup {angularLimitsMax=<2 15 15>, angularLimitsMin=<-2 -15 -15>, angularTarget=<0 0 0>}
      localJointOffset=<1 0 0>
     }]
    relativeSpaceBone {boneName=head}
   }
   n22 = Anim Dynamics - Bone: dyn_hair_t_01(ComponentPose=@n21, Alpha=1, ExternalForce=@n6) {
    angularDampingOverride=0.9999
    angularSpringConstant=300
    bAngularSpring=true
    bOverrideAngularDamping=true
    simulationSpace=RootRelative
    alphaBoolBlend {blendOption=Linear}
    boundBone {boneName=dyn_hair_t_01}
    componentAppliedLinearAccClamp=<10000 10000 10000>
    componentLinearAccScale=<0.7 0.7 2>
    componentLinearVelScale=<0.7 0.7 2>
    physicsBodyDefinitions [{
      boundBone {boneName=dyn_hair_t_01}
      boxExtents=<15 15 15>
      constraintSetup {angularLimitsMax=<2 55 45>, angularLimitsMin=<-2 -55 -45>, angularTarget=<0 0 0>}
      localJointOffset=<0 0 -1>
     }]
    relativeSpaceBone {boneName=head}
   }
   n23 = Anim Dynamics - Bone: dyn_hair_t_02(ComponentPose=@n22, Alpha=1, ExternalForce=@n6) {
    angularDampingOverride=0.9999
    angularSpringConstant=300
    bAngularSpring=true
    bOverrideAngularDamping=true
    simulationSpace=RootRelative
    alphaBoolBlend {blendOption=Linear}
    boundBone {boneName=dyn_hair_t_02}
    componentAppliedLinearAccClamp=<10000 10000 10000>
    componentLinearAccScale=<0.7 0.7 2>
    componentLinearVelScale=<0.7 0.7 2>
    physicsBodyDefinitions [{
      boundBone {boneName=dyn_hair_t_02}
      boxExtents=<15 15 15>
      constraintSetup {angularLimitsMax=<2 55 45>, angularLimitsMin=<-2 -55 -45>, angularTarget=<0 0 0>}
      localJointOffset=<0 0 -1>
     }]
    relativeSpaceBone {boneName=head}
   }
   n24 = Anim Dynamics - Bone: dyn_hair_t_03(ComponentPose=@n23, Alpha=1, ExternalForce=@n6) {
    angularDampingOverride=0.9999
    angularSpringConstant=300
    bAngularSpring=true
    bOverrideAngularDamping=true
    simulationSpace=RootRelative
    alphaBoolBlend {blendOption=Linear}
    boundBone {boneName=dyn_hair_t_03}
    componentAppliedLinearAccClamp=<10000 10000 10000>
    componentLinearAccScale=<0.7 0.7 2>
    componentLinearVelScale=<0.7 0.7 2>
    physicsBodyDefinitions [{
      boundBone {boneName=dyn_hair_t_03}
      boxExtents=<15 15 15>
      constraintSetup {angularLimitsMax=<2 55 45>, angularLimitsMin=<-2 -55 -45>, angularTarget=<0 0 0>}
      localJointOffset=<0 0 -1>
     }]
    relativeSpaceBone {boneName=head}
   }
   n25 = "Constraint - dyn_hair_f_02_01 (PP)"!(ComponentPose=@n24, Alpha=1, ConstraintWeights_0=0.5, ConstraintWeights_1:float=0.5) -> (Pose) {
    alphaBoolBlend {blendOption=Linear}
    boneToModify {boneName=dyn_hair_f_02_01}
    constraintSetup [{targetBone {boneName=dyn_hair_f_01_01}}, {targetBone {boneName=dyn_hair_f_03_01}}]
    constraintWeights [1,1]
   }
   n26 = Get GravityOverride()
   n27 = Anim Dynamics - Bone: dyn_hair_f_01_01(ComponentPose=Local To Component(LocalPose=Use cached pose 'PostCtrlRig'()), Alpha=0, ExternalForce=@n26) {
    angularDampingOverride=0.99
    angularSpringConstant=100
    bOverrideAngularDamping=true
    bUseGravityOverride=true
    simulationSpace=RootRelative
    alphaBoolBlend {blendOption=Linear}
    boundBone {boneName=dyn_hair_f_01_01}
    componentAppliedLinearAccClamp=<10000 10000 10000>
    componentLinearAccScale=<1 1 1>
    componentLinearVelScale=<1 1 1>
    physicsBodyDefinitions [{
      boundBone {boneName=dyn_hair_f_01_01}
      boxExtents=<20 20 20>
      constraintSetup {angularLimitsMax=<1 0 10>, angularLimitsMin=<-1 -20 0>, angularTarget=<0 0 0>}
      localJointOffset=<1 0 0>
     }]
    relativeSpaceBone {boneName=spine_05}
   }
   n28 = Anim Dynamics - Bone: dyn_hair_f_01_02(ComponentPose=@n27, Alpha=0, ExternalForce=@n26) {
    angularDampingOverride=0.99
    angularSpringConstant=75
    bOverrideAngularDamping=true
    bUseSphericalLimits=true
    simulationSpace=RootRelative
    alphaBoolBlend {blendOption=Linear}
    boundBone {boneName=dyn_hair_f_01_02}
    componentAppliedLinearAccClamp=<10000 10000 10000>
    componentLinearAccScale=<2 2 2>
    componentLinearVelScale=<2 2 2>
    physicsBodyDefinitions [{
      boundBone {boneName=dyn_hair_f_01_02}
      boxExtents=<20 20 20>
      constraintSetup {angularLimitsMax=<1 3 15>, angularLimitsMin=<-1 -20 -10>, angularTarget=<0 0 0>}
      localJointOffset=<1 0 0>
     }]
    relativeSpaceBone {boneName=spine_05}
    sphericalLimits [{limitRadius=4.6, drivingBone {boneName=head}, sphereLocalOffset=<5 8 3>}]
   }
   n29 = Anim Dynamics - Bone: dyn_hair_f_01_03(ComponentPose=@n28, Alpha=1, ExternalForce=@n26) {
    angularDampingOverride=0.99
    angularSpringConstant=50
    bOverrideAngularDamping=true
    bUsePlanarLimit=false
    bUseSphericalLimits=true
    simulationSpace=RootRelative
    alphaBoolBlend {blendOption=Linear}
    boundBone {boneName=dyn_hair_f_01_03}
    componentAppliedLinearAccClamp=<10000 10000 10000>
    componentLinearAccScale=<2 2 2>
    componentLinearVelScale=<2 2 2>
    physicsBodyDefinitions [{
      boundBone {boneName=dyn_hair_f_01_03}
      boxExtents=<15 15 15>
      constraintSetup {angularLimitsMax=<1 5 15>, angularLimitsMin=<-1 -15 -5>, angularTarget=<0 0 0>}
      localJointOffset=<1 0 0>
     }]
    relativeSpaceBone {boneName=spine_05}
    sphericalLimits [{limitRadius=4.3, drivingBone {boneName=head}, sphereLocalOffset=<5 7 3>}]
   }
   n30 = Anim Dynamics - Bone: dyn_hair_f_01_05(ComponentPose=@n29, Alpha=1, ExternalForce=@n26) {
    angularDampingOverride=0.9
    angularSpringConstant=50
    bOverrideAngularDamping=true
    bUsePlanarLimit=false
    bUseSphericalLimits=true
    simulationSpace=RootRelative
    alphaBoolBlend {blendOption=Linear}
    boundBone {boneName=dyn_hair_f_01_05}
    componentAppliedLinearAccClamp=<10000 10000 10000>
    componentLinearAccScale=<2 2 2>
    componentLinearVelScale=<2 2 2>
    physicsBodyDefinitions [{
      boundBone {boneName=dyn_hair_f_01_05}
      boxExtents=<15 15 15>
      constraintSetup {angularLimitsMax=<1 5 15>, angularLimitsMin=<-1 -15 -15>, angularTarget=<0 0 0>}
      localJointOffset=<1 0 0>
     }]
    relativeSpaceBone {boneName=spine_05}
    sphericalLimits [{limitRadius=4, drivingBone {boneName=head}, sphereLocalOffset=<2 9 3>}]
   }
   n31 = Anim Dynamics - Bone: dyn_hair_f_03_01(ComponentPose=@n30, Alpha=1, ExternalForce=@n26) {
    angularDampingOverride=0.99
    angularSpringConstant=100
    bOverrideAngularDamping=true
    simulationSpace=RootRelative
    alphaBoolBlend {blendOption=Linear}
    boundBone {boneName=dyn_hair_f_03_01}
    componentAppliedLinearAccClamp=<10000 10000 10000>
    componentLinearAccScale=<1 1 1>
    componentLinearVelScale=<1 1 1>
    physicsBodyDefinitions [{
      boundBone {boneName=dyn_hair_f_03_01}
      boxExtents=<20 20 20>
      constraintSetup {angularLimitsMax=<1 0 10>, angularLimitsMin=<-1 -20 -10>, angularTarget=<0 0 0>}
      localJointOffset=<1 0 0>
     }]
    planarLimits [{drivingBone {boneName=head}, planeTransform {location=<5 0 2>, rotation=<rot 0 0 0>}}]
    relativeSpaceBone {boneName=spine_05}
    sphericalLimits [{limitRadius=4, drivingBone {boneName=head}, sphereLocalOffset=<5 5 0>}]
   }
   n32 = Anim Dynamics - Bone: dyn_hair_f_03_02(ComponentPose=@n31, Alpha=1, ExternalForce=@n26) {
    angularDampingOverride=0.99
    angularSpringConstant=75
    bOverrideAngularDamping=true
    simulationSpace=RootRelative
    alphaBoolBlend {blendOption=Linear}
    boundBone {boneName=dyn_hair_f_03_02}
    componentAppliedLinearAccClamp=<10000 10000 10000>
    componentLinearAccScale=<2 2 2>
    componentLinearVelScale=<2 2 2>
    physicsBodyDefinitions [{
      boundBone {boneName=dyn_hair_f_03_02}
      boxExtents=<20 20 20>
      constraintSetup {angularLimitsMax=<1 3 15>, angularLimitsMin=<-1 -20 -15>, angularTarget=<0 0 0>}
      localJointOffset=<1 0 0>
     }]
    planarLimits [{drivingBone {boneName=head}, planeTransform {location=<5 0 2.5>, rotation=<rot 0 0 0>}}]
    relativeSpaceBone {boneName=spine_05}
    sphericalLimits [{limitRadius=6, drivingBone {boneName=head}, sphereLocalOffset=<5 5 0>}]
   }
   n33 = Anim Dynamics - Bone: dyn_hair_f_03_03(ComponentPose=@n32, Alpha=1, ExternalForce=@n26) {
    angularDampingOverride=0.99
    angularSpringConstant=50
    bOverrideAngularDamping=true
    bUsePlanarLimit=false
    bUseSphericalLimits=true
    simulationSpace=RootRelative
    alphaBoolBlend {blendOption=Linear}
    boundBone {boneName=dyn_hair_f_03_03}
    componentAppliedLinearAccClamp=<10000 10000 10000>
    componentLinearAccScale=<2 2 2>
    componentLinearVelScale=<2 2 2>
    physicsBodyDefinitions [{
      boundBone {boneName=dyn_hair_f_03_03}
      boxExtents=<15 15 15>
      constraintSetup {angularLimitsMax=<1 5 15>, angularLimitsMin=<-1 -15 -15>, angularTarget=<0 0 0>}
      localJointOffset=<1 0 0>
     }]
    relativeSpaceBone {boneName=spine_05}
    sphericalLimits [{limitRadius=4, drivingBone {boneName=head}, sphereLocalOffset=<5 7 3>}]
   }
   n34 = Anim Dynamics - Bone: dyn_hair_f_03_04(ComponentPose=@n33, Alpha=1, ExternalForce=@n26) {
    angularDampingOverride=0.9
    angularSpringConstant=50
    bOverrideAngularDamping=true
    bUsePlanarLimit=false
    bUseSphericalLimits=true
    simulationSpace=RootRelative
    alphaBoolBlend {blendOption=Linear}
    boundBone {boneName=dyn_hair_f_03_04}
    componentAppliedLinearAccClamp=<10000 10000 10000>
    componentLinearAccScale=<2 2 2>
    componentLinearVelScale=<2 2 2>
    physicsBodyDefinitions [{
      boundBone {boneName=dyn_hair_f_03_04}
      boxExtents=<15 15 15>
      constraintSetup {angularLimitsMax=<1 5 15>, angularLimitsMin=<-1 -15 -15>, angularTarget=<0 0 0>}
      localJointOffset=<1 0 0>
     }]
    relativeSpaceBone {boneName=spine_05}
    sphericalLimits [{limitRadius=4, drivingBone {boneName=head}, sphereLocalOffset=<2 5 3>}]
   }
   n35 = Anim Dynamics - Bone: dyn_hair_f_04_01(ComponentPose=@n34, Alpha=1, ExternalForce=@n26) {
    angularDampingOverride=0.99
    angularSpringConstant=100
    bOverrideAngularDamping=true
    simulationSpace=RootRelative
    alphaBoolBlend {blendOption=Linear}
    boundBone {boneName=dyn_hair_f_04_01}
    componentAppliedLinearAccClamp=<10000 10000 10000>
    componentLinearAccScale=<1 1 1>
    componentLinearVelScale=<1 1 1>
    physicsBodyDefinitions [{
      boundBone {boneName=dyn_hair_f_04_01}
      boxExtents=<20 20 20>
      constraintSetup {angularLimitsMax=<1 0 10>, angularLimitsMin=<-1 -20 -10>, angularTarget=<0 0 0>}
      localJointOffset=<1 0 0>
     }]
    planarLimits [{drivingBone {boneName=head}, planeTransform {location=<5 0 4>, rotation=<rot 0 0 0>}}]
    relativeSpaceBone {boneName=spine_05}
   }
   n36 = Anim Dynamics - Bone: dyn_hair_f_04_02(ComponentPose=@n35, Alpha=1, ExternalForce=@n26) {
    angularDampingOverride=0.99
    angularSpringConstant=75
    bOverrideAngularDamping=true
    simulationSpace=RootRelative
    alphaBoolBlend {blendOption=Linear}
    boundBone {boneName=dyn_hair_f_04_02}
    componentAppliedLinearAccClamp=<10000 10000 10000>
    componentLinearAccScale=<2 2 2>
    componentLinearVelScale=<2 2 2>
    physicsBodyDefinitions [{
      boundBone {boneName=dyn_hair_f_04_02}
      boxExtents=<20 20 20>
      constraintSetup {angularLimitsMax=<1 3 15>, angularLimitsMin=<-1 -20 -15>, angularTarget=<0 0 0>}
      localJointOffset=<1 0 0>
     }]
    planarLimits [{drivingBone {boneName=head}, planeTransform {location=<5 0 7>, rotation=<rot 0 0 0>}}]
    relativeSpaceBone {boneName=spine_05}
   }
   n37 = Anim Dynamics - Bone: dyn_hair_f_04_03(ComponentPose=@n36, Alpha=1, ExternalForce=@n26) {
    angularDampingOverride=0.99
    angularSpringConstant=50
    bOverrideAngularDamping=true
    simulationSpace=RootRelative
    alphaBoolBlend {blendOption=Linear}
    boundBone {boneName=dyn_hair_f_04_03}
    componentAppliedLinearAccClamp=<10000 10000 10000>
    componentLinearAccScale=<2 2 2>
    componentLinearVelScale=<2 2 2>
    physicsBodyDefinitions [{
      boundBone {boneName=dyn_hair_f_04_03}
      boxExtents=<15 15 15>
      constraintSetup {angularLimitsMax=<1 5 15>, angularLimitsMin=<-1 -15 -15>, angularTarget=<0 0 0>}
      localJointOffset=<1 0 0>
     }]
    relativeSpaceBone {boneName=spine_05}
   }
   n38 = Anim Dynamics - Bone: dyn_hair_f_04_04(ComponentPose=@n37, Alpha=1, ExternalForce=@n26) {
    angularDampingOverride=0.9
    angularSpringConstant=50
    bOverrideAngularDamping=true
    simulationSpace=RootRelative
    alphaBoolBlend {blendOption=Linear}
    boundBone {boneName=dyn_hair_f_04_04}
    componentAppliedLinearAccClamp=<10000 10000 10000>
    componentLinearAccScale=<2 2 2>
    componentLinearVelScale=<2 2 2>
    physicsBodyDefinitions [{
      boundBone {boneName=dyn_hair_f_04_04}
      boxExtents=<15 15 15>
      constraintSetup {angularLimitsMax=<1 5 15>, angularLimitsMin=<-1 -15 -15>, angularTarget=<0 0 0>}
      localJointOffset=<1 0 0>
     }]
    relativeSpaceBone {boneName=spine_05}
   }
   n39 = Anim Dynamics - Bone: dyn_hair_f_05_01(ComponentPose=@n38, Alpha=1, ExternalForce=@n26) {
    angularDampingOverride=0.999
    angularSpringConstant=50
    bAngularSpring=true
    bOverrideAngularDamping=true
    simulationSpace=RootRelative
    alphaBoolBlend {blendOption=Linear}
    boundBone {boneName=dyn_hair_f_05_01}
    componentAppliedLinearAccClamp=<10000 10000 10000>
    componentLinearAccScale=<0.2 0.2 0.2>
    componentLinearVelScale=<0.2 0.2 0.2>
    physicsBodyDefinitions [{
      boundBone {boneName=dyn_hair_f_05_01}
      boxExtents=<20 20 20>
      constraintSetup {angularLimitsMax=<2 3 15>, angularLimitsMin=<-2 -10 -15>, angularTarget=<0 0 0>}
      localJointOffset=<1 0 0>
     }]
    relativeSpaceBone {boneName=head}
   }
   n40 = Anim Dynamics - Bone: dyn_hair_f_05_02(ComponentPose=@n39, Alpha=1, ExternalForce=@n26) {
    angularDampingOverride=0.99
    angularSpringConstant=50
    bAngularSpring=true
    bOverrideAngularDamping=true
    simulationSpace=RootRelative
    alphaBoolBlend {blendOption=Linear}
    boundBone {boneName=dyn_hair_f_05_02}
    componentAppliedLinearAccClamp=<10000 10000 10000>
    componentLinearAccScale=<0.3 0.3 0.3>
    componentLinearVelScale=<0.3 0.3 0.3>
    physicsBodyDefinitions [{
      boundBone {boneName=dyn_hair_f_05_02}
      boxExtents=<15 15 15>
      constraintSetup {angularLimitsMax=<2 20 20>, angularLimitsMin=<-2 -20 -20>, angularTarget=<0 0 0>}
      localJointOffset=<1 0 0>
     }]
    relativeSpaceBone {boneName=head}
   }
   n41 = Anim Dynamics - Bone: dyn_hair_f_02_02(ComponentPose=@n40, Alpha=1, ExternalForce=@n26) {
    angularDampingOverride=0.99
    angularSpringConstant=50
    bAngularSpring=true
    bOverrideAngularDamping=true
    simulationSpace=RootRelative
    alphaBoolBlend {blendOption=Linear}
    boundBone {boneName=dyn_hair_f_02_02}
    componentAppliedLinearAccClamp=<10000 10000 10000>
    componentLinearAccScale=<0.3 0.3 0.3>
    componentLinearVelScale=<0.3 0.3 0.3>
    physicsBodyDefinitions [{
      boundBone {boneName=dyn_hair_f_02_02}
      boxExtents=<20 20 20>
      constraintSetup {angularLimitsMax=<2 15 15>, angularLimitsMin=<-2 -15 -15>, angularTarget=<0 0 0>}
      localJointOffset=<1 0 0>
     }]
    relativeSpaceBone {boneName=head}
   }
   n42 = Anim Dynamics - Bone: dyn_hair_t_01(ComponentPose=@n41, Alpha=1, ExternalForce=@n26) {
    angularDampingOverride=0.9999
    angularSpringConstant=300
    bAngularSpring=true
    bOverrideAngularDamping=true
    simulationSpace=RootRelative
    alphaBoolBlend {blendOption=Linear}
    boundBone {boneName=dyn_hair_t_01}
    componentAppliedLinearAccClamp=<10000 10000 10000>
    componentLinearAccScale=<0.7 0.7 2>
    componentLinearVelScale=<0.7 0.7 2>
    physicsBodyDefinitions [{
      boundBone {boneName=dyn_hair_t_01}
      boxExtents=<15 15 15>
      constraintSetup {angularLimitsMax=<2 55 45>, angularLimitsMin=<-2 -55 -45>, angularTarget=<0 0 0>}
      localJointOffset=<0 0 -1>
     }]
    relativeSpaceBone {boneName=head}
   }
   n43 = Anim Dynamics - Bone: dyn_hair_t_02(ComponentPose=@n42, Alpha=1, ExternalForce=@n26) {
    angularDampingOverride=0.9999
    angularSpringConstant=300
    bAngularSpring=true
    bOverrideAngularDamping=true
    simulationSpace=RootRelative
    alphaBoolBlend {blendOption=Linear}
    boundBone {boneName=dyn_hair_t_02}
    componentAppliedLinearAccClamp=<10000 10000 10000>
    componentLinearAccScale=<0.7 0.7 2>
    componentLinearVelScale=<0.7 0.7 2>
    physicsBodyDefinitions [{
      boundBone {boneName=dyn_hair_t_02}
      boxExtents=<15 15 15>
      constraintSetup {angularLimitsMax=<2 55 45>, angularLimitsMin=<-2 -55 -45>, angularTarget=<0 0 0>}
      localJointOffset=<0 0 -1>
     }]
    relativeSpaceBone {boneName=head}
   }
   n44 = Anim Dynamics - Bone: dyn_hair_t_03(ComponentPose=@n43, Alpha=1, ExternalForce=@n26) {
    angularDampingOverride=0.9999
    angularSpringConstant=300
    bAngularSpring=true
    bOverrideAngularDamping=true
    simulationSpace=RootRelative
    alphaBoolBlend {blendOption=Linear}
    boundBone {boneName=dyn_hair_t_03}
    componentAppliedLinearAccClamp=<10000 10000 10000>
    componentLinearAccScale=<0.7 0.7 2>
    componentLinearVelScale=<0.7 0.7 2>
    physicsBodyDefinitions [{
      boundBone {boneName=dyn_hair_t_03}
      boxExtents=<15 15 15>
      constraintSetup {angularLimitsMax=<2 55 45>, angularLimitsMin=<-2 -55 -45>, angularTarget=<0 0 0>}
      localJointOffset=<0 0 -1>
     }]
    relativeSpaceBone {boneName=head}
   }
   n45 = "Constraint - dyn_hair_f_01_01 (P)"(ComponentPose=@n44, Alpha=1, ConstraintWeights_0=1) {
    alphaBoolBlend {blendOption=Linear}
    boneToModify {boneName=dyn_hair_f_01_01}
    constraintSetup [{targetBone {boneName=dyn_hair_f_03_01}}]
    constraintWeights [1]
   }
   n46 = "Constraint - dyn_hair_f_01_02 (P)"(ComponentPose=@n45, Alpha=1, ConstraintWeights_0=1) {
    alphaBoolBlend {blendOption=Linear}
    boneToModify {boneName=dyn_hair_f_01_02}
    constraintSetup [{targetBone {boneName=dyn_hair_f_03_02}}]
    constraintWeights [1]
   }
   Output Pose(Result=Blend Poses by bool(BlendPose_0=Component To Local(ComponentPose=@n46), BlendPose_1=Component To Local(ComponentPose=@n25)) {blendType=Linear}) {layerGroup=None}
   n47 = Control Rig!(Source=Copy Pose From Mesh() {bUseAttachedParent=true}, Alpha, RbanCtrl:bool=false) -> (Pose) {
    alphaBoolBlend {blendOption=Linear}
    controlRigAssetReference {blueprintRigClass=&/M_MED_LazarusLens_FaceAcc_CtrlRig.M_MED_LazarusLens_FaceAcc_CtrlRig_C}
   }
   Save cached pose 'PostCtrlRig'(Pose=@n47) {cacheName=PostCtrlRig}
  ]}
 EventGraph {nodes [
   Event BlueprintUpdateAnimation()
   > n1 = Gravity Override State Function DEPRECATED(CharacterPartAnimInstance=Self-Reference(), GameTimeInSeconds=GetGameTimeInSeconds(), GravityOverrideParams=Get Gravity Override Params())
   > Set GravityOverride(GravityOverride=@n1)
  ]}
]
variables [
 Gravity Override Params {type=GravityOverrideParamsStruct}
 GravityOverride {type=Vector}
]
